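# Sound cue for the foley played when aiming down sights with the Skyfire knockback launcher.
# A single wave runs through a modulator that randomizes pitch and volume, feeding a quality-level node that connects it only to the first two levels and leaves the other two empty.
# The cue plays quietly in the weapon foley class with reload attenuation, overrides concurrency to one instance, and has virtualization disabled.
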SoundCue /Skyfire/Gameplay/KnockbackWeapon/Audio/Knockback_Launcher_ADS_In_Cue
root=QualityLevel_0
nodes [
 {id=Modulator_0, children [WavePlayer_0], properties {pitchMax=1.2, pitchMin=0.9, volumeMax=1, volumeMin=0.65}}
 {id=QualityLevel_0, children [Modulator_0,Modulator_0,null,null]}
 {id=WavePlayer_0, properties {soundWaveAssetPtr=&/Knockback_Launcher_ADS_In_01}}
]
properties {
 attenuationSettings=&Fort_Weapon_Reload_Att
 bOverrideConcurrency=true
 maxDistance=2250
 priority=0
 soundClassObject=&Fort_WeaponFoley
 virtualizationMode=Disabled
 volumeMultiplier=0.2
 concurrencyOverrides {maxCount=1}
}
