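# Sound cue for the knockback launcher's ADS-out (leaving aim-down-sights) weapon foley.
# A quality-level switch sends its first two slots to one Modulator that randomizes pitch on a single wave, leaving the other two slots empty.
# It plays quietly and slightly pitched down in the weapon foley class, using reload attenuation, a single-instance concurrency override and no virtualization.
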SoundCue /Skyfire/Gameplay/KnockbackWeapon/Audio/Knockback_Launcher_ADS_Out_Cue
root=QualityLevel_0
nodes [
 {id=Modulator_0, children [WavePlayer_0], properties {pitchMax=1.1}}
 {id=QualityLevel_0, children [Modulator_0,Modulator_0,null,null]}
 {id=WavePlayer_0, properties {soundWaveAssetPtr=&/Knockback_Launcher_ADS_Out_01}}
]
properties {
 attenuationSettings=&Fort_Weapon_Reload_Att
 bOverrideConcurrency=true
 maxDistance=2250
 pitchMultiplier=0.9
 priority=0
 soundClassObject=&Fort_WeaponFoley
 virtualizationMode=Disabled
 volumeMultiplier=0.1
 concurrencyOverrides {maxCount=1}
}
