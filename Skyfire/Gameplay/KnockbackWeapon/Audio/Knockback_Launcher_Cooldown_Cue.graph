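# Sound cue for the knockback launcher's cooldown, playing a single overheat wave.
# The wave player feeds the output directly, routed through the weapon foley sound class with weapon reload attenuation at reduced volume.
# An overridden concurrency limits it to one instance at a time, and virtualization is disabled.
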SoundCue /Skyfire/Gameplay/KnockbackWeapon/Audio/Knockback_Launcher_Cooldown_Cue
root=WavePlayer_0
nodes [{id=WavePlayer_0, properties {soundWaveAssetPtr=&/Knockback_Launcher_Overheat_01}}]
properties {
 attenuationSettings=&Fort_Weapon_Reload_Att
 bOverrideConcurrency=true
 maxDistance=2250
 priority=75
 soundClassObject=&Fort_WeaponFoley
 virtualizationMode=Disabled
 volumeMultiplier=0.501
 concurrencyOverrides {maxCount=1}
}
