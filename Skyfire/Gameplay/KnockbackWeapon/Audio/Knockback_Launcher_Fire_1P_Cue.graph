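# First-person fire sound cue for the Knockback Launcher weapon, layering the shot with a recharge tail.
# One mixer input picks randomly among four close fire waves with slight pitch variation.
# A second, much quieter input plays the recharge wave after a fixed quarter-second delay with varied volume; a third input is empty.
# High priority and a one-instance, per-owner stop-oldest concurrency override make each new shot cut off the last.
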SoundCue /Skyfire/Gameplay/KnockbackWeapon/Audio/Knockback_Launcher_Fire_1P_Cue
root=Mixer_0
nodes [
 {id=Delay_0, children [Modulator_0], properties {delayMax=0.25, delayMin=0.25}}
 {id=Mixer_0, children [Modulator_3,Delay_0,null], properties {inputVolume [0.95,0.1,0.75]}}
 {id=Modulator_0, children [WavePlayer_8], properties {volumeMax=1, volumeMin=0.6}}
 {id=Modulator_3, children [Random_0], properties {pitchMax=1.1, pitchMin=0.9, volumeMax=1, volumeMin=1}}
 {id=Random_0, children [WavePlayer_0,WavePlayer_3,WavePlayer_5,WavePlayer_6], properties {weights [1,1,1,1]}}
 {id=WavePlayer_0, properties {soundWaveAssetPtr=&/Knockback_Launcher_Fire_Close_01}}
 {id=WavePlayer_3, properties {soundWaveAssetPtr=&/Knockback_Launcher_Fire_Close_04}}
 {id=WavePlayer_5, properties {soundWaveAssetPtr=&/Knockback_Launcher_Fire_Close_03}}
 {id=WavePlayer_6, properties {soundWaveAssetPtr=&/Knockback_Launcher_Fire_Close_02}}
 {id=WavePlayer_8, properties {soundWaveAssetPtr=&/Knockback_Launcher_Fire_Recharge_01}}
]
properties {
 attenuationSettings=&Fort_Weapon_Projectile_FirstPerson_Athena_Att
 bBypassVolumeScaleForPriority=true
 bOverrideConcurrency=true
 maxDistance=2097152
 priority=80
 soundClassObject=&Fort_WeaponRanged_Fire_1P_PSM
 volumeMultiplier=0.95
 attenuationOverrides {falloffDistance=5000, attenuationShapeExtents {x=1000}}
 concurrencyOverrides {bLimitToOwner=true, maxCount=1, resolutionRule=StopOldest}
}
