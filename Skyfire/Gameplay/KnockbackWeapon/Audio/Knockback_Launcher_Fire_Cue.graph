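# Sound cue for firing the Knockback Launcher, mixing close, mid and distant layers.
# Each layer has its own attenuation and sound class and randomly picks one of four variations.
# The close layer adds pitch variation and a short-delayed recharge sound.
# A concurrency override caps simultaneous plays per owner, stopping the quietest.
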SoundCue /Skyfire/Gameplay/KnockbackWeapon/Audio/Knockback_Launcher_Fire_Cue
root=Mixer_17
nodes [
 {
  id=Attenuation_0
  children [Modulator_0]
  properties {attenuationSettings=&Fort_Weapon_Projectile_Medium_Mono_Athena_Att}
 }
 {
  id=Attenuation_34
  children [Modulator_1]
  properties {attenuationSettings=&Fort_Weapon_Projectile_Distant_Mono_Athena_Att}
 }
 {id=Attenuation_49, children [Mixer_0], properties {attenuationSettings=&Fort_Weapon_Projectile_Close_Athena_Att}}
 {id=Delay_0, children [Modulator_3], properties {delayMax=0.25, delayMin=0.25}}
 {id=Mixer_0, children [Modulator_2,Delay_0,null], properties {inputVolume [0.75,0.1,0.45]}}
 {id=Mixer_17, children [SoundClass_32, SoundClass_0, SoundClass_35], properties {inputVolume [0.9,0.5,0.75]}}
 {id=Modulator_0, children [Random_4], properties {volumeMax=1, volumeMin=1}}
 {id=Modulator_1, children [Random_3], properties {volumeMax=1, volumeMin=1}}
 {id=Modulator_2, children [Random_0], properties {pitchMax=1.1, pitchMin=0.9, volumeMax=1, volumeMin=1}}
 {id=Modulator_3, children [WavePlayer_22], properties {volumeMax=1, volumeMin=0.6}}
 {id=Random_0, children [WavePlayer_0,WavePlayer_5,WavePlayer_6,WavePlayer_7], properties {weights [1,1,1,1]}}
 {id=Random_3, children [WavePlayer_10, WavePlayer_11, WavePlayer_12, WavePlayer_13], properties {weights [1,1,1,1]}}
 {id=Random_4, children [WavePlayer_15, WavePlayer_16, WavePlayer_17, WavePlayer_18], properties {weights [1,1,1,1]}}
 {
  id=SoundClass_0
  children [Attenuation_0]
  properties {soundClassOverride=&Fort_WeaponRanged_Fire_Distant}
 }
 {
  id=SoundClass_32
  children [Attenuation_49]
  properties {soundClassOverride=&Fort_WeaponRanged_Fire_3P_PSM}
 }
 {
  id=SoundClass_35
  children [Attenuation_34]
  properties {soundClassOverride=&Fort_WeaponRanged_Fire_Distant}
 }
 {id=WavePlayer_0, properties {soundWaveAssetPtr=&/Knockback_Launcher_Fire_Close_01}}
 {id=WavePlayer_10, properties {soundWaveAssetPtr=&/Knockback_Launcher_Fire_Distant_01}}
 {id=WavePlayer_11, properties {soundWaveAssetPtr=&/Knockback_Launcher_Fire_Distant_02}}
 {id=WavePlayer_12, properties {soundWaveAssetPtr=&/Knockback_Launcher_Fire_Distant_03}}
 {id=WavePlayer_13, properties {soundWaveAssetPtr=&/Knockback_Launcher_Fire_Distant_04}}
 {id=WavePlayer_15, properties {soundWaveAssetPtr=&/Knockback_Launcher_Fire_Mid_01}}
 {id=WavePlayer_16, properties {soundWaveAssetPtr=&/Knockback_Launcher_Fire_Mid_02}}
 {id=WavePlayer_17, properties {soundWaveAssetPtr=&/Knockback_Launcher_Fire_Mid_03}}
 {id=WavePlayer_18, properties {soundWaveAssetPtr=&/Knockback_Launcher_Fire_Mid_04}}
 {id=WavePlayer_22, properties {soundWaveAssetPtr=&/Knockback_Launcher_Fire_Recharge_01}}
 {id=WavePlayer_5, properties {soundWaveAssetPtr=&/Knockback_Launcher_Fire_Close_04}}
 {id=WavePlayer_6, properties {soundWaveAssetPtr=&/Knockback_Launcher_Fire_Close_03}}
 {id=WavePlayer_7, properties {soundWaveAssetPtr=&/Knockback_Launcher_Fire_Close_02}}
]
properties {
 bOverrideConcurrency=true
 maxDistance=27500
 priority=75
 volumeMultiplier=1
 attenuationOverrides {falloffDistance=5000, attenuationShapeExtents {x=1000}}
 concurrencyOverrides {bLimitToOwner=true, maxCount=3, resolutionRule=StopQuietest}
}
