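# Sound cue for the Knockback Launcher impact in the Skyfire knockback weapon, currently containing no sound nodes or root, so it has no audio source.
# It carries only playback settings: the Fort_Impacts sound class and a long-range opponent bullet-impact attenuation asset.
# Its concurrency override allows one instance at a time and prevents new ones, with virtualization disabled.
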SoundCue /Skyfire/Gameplay/KnockbackWeapon/Audio/Knockback_Launcher_Impact_Cue
root=null
nodes []
properties {
 attenuationSettings=/Game/Sounds/Attenuation/Fort_Bullet_ImpactOpponent_Lng_Att
 bOverrideConcurrency=true
 maxDistance=5500
 soundClassObject=&Fort_Impacts
 virtualizationMode=Disabled
 volumeMultiplier=1
 concurrencyOverrides {maxCount=1, resolutionRule=PreventNew}
}
