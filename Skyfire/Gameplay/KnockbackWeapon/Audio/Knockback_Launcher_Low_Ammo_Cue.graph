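# Sound cue for the Knockback Launcher's low-ammo feedback, playing a single out-of-ammo wave at a reduced volume.
# The wave feeds a Modulator that adds a very slight random pitch variation.
# Its own concurrency override allows one instance at a time, stopping the oldest, and virtualization is disabled.
# It uses the third-person ranged weapon fire sound class and overrides attenuation falloff distance.
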SoundCue /Skyfire/Gameplay/KnockbackWeapon/Audio/Knockback_Launcher_Low_Ammo_Cue
root=Modulator_1
nodes [
 {id=Modulator_1, children [WavePlayer_0], properties {pitchMax=1, pitchMin=0.99, volumeMax=1}}
 {id=WavePlayer_0, properties {soundWaveAssetPtr=&/Knockback_Launcher_OutOfAmmo_01}}
]
properties {
 bOverrideConcurrency=true
 maxDistance=2097152
 priority=0
 soundClassObject=&Fort_WeaponRanged_Fire_3P_PSM
 virtualizationMode=Disabled
 volumeMultiplier=0.1
 attenuationOverrides {falloffDistance=1600}
 concurrencyOverrides {maxCount=1, resolutionRule=StopOldest}
}
