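# Sound cue for picking up the knockback launcher weapon.
# A single wave player feeds a default Modulator, reusing the Gravy Goblin launcher pickup sound from MotherGameplay.
# It uses weapon pickup attenuation, the loot collect sound class and a reduced volume.
# It references the shared pickup concurrency set and carries concurrency override values for a single instance that prevents new plays.
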SoundCue /Skyfire/Gameplay/KnockbackWeapon/Audio/Knockback_Launcher_Pickup_Cue
root=Modulator_0
nodes [
 {id=Modulator_0, children [WavePlayer_2]}
 {id=WavePlayer_2, properties {soundWaveAssetPtr=/MotherGameplay/Items/GravyGoblin/Assets/Audio/Launcher_Gravy_Goblin_Pickup_01}}
]
properties {
 attenuationSettings=&Fort_Weapon_Pickup_Att
 maxDistance=3500
 soundClassObject=&Fort_Loot_Collect
 volumeMultiplier=0.5538
 concurrencyOverrides {maxCount=1, resolutionRule=PreventNew}
 concurrencySet [&PickupMaximum_Oldest]
}
