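# Sound cue for the knockback launcher projectile explosion, layering close, medium and distant variants by distance.
# A mixer blends three branches, each with its own sound class, explosion attenuation (close and distant override falloff and shape), pitch modulator and random pick of wave files.
# Close uses a 3P weapon-fire class, while medium and distant share a distant-fire class.
# Cue-level concurrency caps simultaneous instances, stopping the quietest, and virtualization is disabled.
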SoundCue /Skyfire/Gameplay/KnockbackWeapon/Audio/Knockback_Launcher_Prj_Explo_Cue
root=Mixer_2
nodes [
 {
  id=Attenuation_3
  children [Modulator_5]
  properties {
   attenuationSettings=/Game/Sounds/Attenuation/Fort_Explosion_Close_Athena_Att
   attenuationOverrides {falloffDistance=900, attenuationShapeExtents {x=100}}
  }
 }
 {id=Attenuation_5, children [Modulator_6], properties {attenuationSettings=/Game/Sounds/Attenuation/Fort_Explosion_Medium_Athena_Att}}
 {
  id=Attenuation_6
  children [Modulator_8]
  properties {
   attenuationSettings=/Game/Sounds/Attenuation/Fort_Explosion_Distant_Athena_Att
   attenuationOverrides {falloffDistance=3000, attenuationShapeExtents {x=1000}}
  }
 }
 {id=Mixer_2, children [SoundClass_5,SoundClass_6,SoundClass_7], properties {inputVolume [0.8,0.5,1.1]}}
 {id=Modulator_5, children [Random_10], properties {pitchMax=1.2, pitchMin=0.9, volumeMax=1, volumeMin=1}}
 {id=Modulator_6, children [Random_11], properties {pitchMax=1.3}}
 {id=Modulator_8, children [Random_12], properties {pitchMax=1.2, pitchMin=0.8}}
 {
  id=Random_10
  children [WavePlayer_36, WavePlayer_37, WavePlayer_38, WavePlayer_39, WavePlayer_40, WavePlayer_41]
  properties {weights [1,1,1,1,1,1]}
 }
 {
  id=Random_11
  children [WavePlayer_42, WavePlayer_43, WavePlayer_44, WavePlayer_45, WavePlayer_46, WavePlayer_47]
  properties {weights [1,1,1,1,1,1]}
 }
 {
  id=Random_12
  children [WavePlayer_48, WavePlayer_49, WavePlayer_50, WavePlayer_51, WavePlayer_52, WavePlayer_53, WavePlayer_54, WavePlayer_55, WavePlayer_56]
  properties {weights [1,1,1,1,1,1,1,1,1]}
 }
 {
  id=SoundClass_5
  children [Attenuation_3]
  properties {soundClassOverride=&Fort_WeaponRanged_Fire_3P_PSM}
 }
 {
  id=SoundClass_6
  children [Attenuation_5]
  properties {soundClassOverride=&Fort_WeaponRanged_Fire_Distant}
 }
 {
  id=SoundClass_7
  children [Attenuation_6]
  properties {soundClassOverride=&Fort_WeaponRanged_Fire_Distant}
 }
 {id=WavePlayer_36, properties {soundWaveAssetPtr=&/Knockback_Launcher_Prj_Explo_Close_01}}
 {id=WavePlayer_37, properties {soundWaveAssetPtr=&/Knockback_Launcher_Prj_Explo_Close_06}}
 {id=WavePlayer_38, properties {soundWaveAssetPtr=&/Knockback_Launcher_Prj_Explo_Close_05}}
 {id=WavePlayer_39, properties {soundWaveAssetPtr=&/Knockback_Launcher_Prj_Explo_Close_04}}
 {id=WavePlayer_40, properties {soundWaveAssetPtr=&/Knockback_Launcher_Prj_Explo_Close_03}}
 {id=WavePlayer_41, properties {soundWaveAssetPtr=&/Knockback_Launcher_Prj_Explo_Close_02}}
 {id=WavePlayer_42, properties {soundWaveAssetPtr=&/Knockback_Launcher_Prj_Explo_Mid_06}}
 {id=WavePlayer_43, properties {soundWaveAssetPtr=&/Knockback_Launcher_Prj_Explo_Mid_05}}
 {id=WavePlayer_44, properties {soundWaveAssetPtr=&/Knockback_Launcher_Prj_Explo_Mid_04}}
 {id=WavePlayer_45, properties {soundWaveAssetPtr=&/Knockback_Launcher_Prj_Explo_Mid_03}}
 {id=WavePlayer_46, properties {soundWaveAssetPtr=&/Knockback_Launcher_Prj_Explo_Mid_02}}
 {id=WavePlayer_47, properties {soundWaveAssetPtr=&/Knockback_Launcher_Prj_Explo_Mid_01}}
 {id=WavePlayer_48, properties {soundWaveAssetPtr=&/Knockback_Launcher_Prj_Explo_Distant_09}}
 {id=WavePlayer_49, properties {soundWaveAssetPtr=&/Knockback_Launcher_Prj_Explo_Distant_08}}
 {id=WavePlayer_50, properties {soundWaveAssetPtr=&/Knockback_Launcher_Prj_Explo_Distant_07}}
 {id=WavePlayer_51, properties {soundWaveAssetPtr=&/Knockback_Launcher_Prj_Explo_Distant_06}}
 {id=WavePlayer_52, properties {soundWaveAssetPtr=&/Knockback_Launcher_Prj_Explo_Distant_05}}
 {id=WavePlayer_53, properties {soundWaveAssetPtr=&/Knockback_Launcher_Prj_Explo_Distant_04}}
 {id=WavePlayer_54, properties {soundWaveAssetPtr=&/Knockback_Launcher_Prj_Explo_Distant_03}}
 {id=WavePlayer_55, properties {soundWaveAssetPtr=&/Knockback_Launcher_Prj_Explo_Distant_02}}
 {id=WavePlayer_56, properties {soundWaveAssetPtr=&/Knockback_Launcher_Prj_Explo_Distant_01}}
]
properties {
 bOverrideConcurrency=true
 maxDistance=27500
 priority=80
 virtualizationMode=Disabled
 volumeMultiplier=1
 attenuationOverrides {reverbWetLevelMax=0.7, reverbWetLevelMin=0.2}
 concurrencyOverrides {maxCount=3, resolutionRule=StopQuietest, volumeScale=0.5}
}
