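# First-person sound cue for a Knockback Launcher projectile hit pushing the player back, inferred from the asset name.
# It mixes two delayed layers: an early close impact and a later quad whoosh, each randomly picking one of three waves with pitch variation.
# Concurrency is overridden to a single instance that stops the oldest, with high priority.
# Despite the 1P name, it is assigned to a third-person ranged weapon fire sound class.
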SoundCue /Skyfire/Gameplay/KnockbackWeapon/Audio/Knockback_Launcher_Prj_Hit_Player_Push_Back_1P_Cue
root=Mixer_0
nodes [
 {id=Delay_0, children [Modulator_3], properties {delayMax=0.1, delayMin=0.1}}
 {id=Delay_1, children [Modulator_0], properties {delayMax=0.45, delayMin=0.45}}
 {id=Mixer_0, children [Delay_0,Delay_1], properties {inputVolume [0.65,0.5]}}
 {id=Modulator_0, children [Random_1], properties {pitchMax=1.1, pitchMin=0.9, volumeMax=1, volumeMin=1}}
 {id=Modulator_3, children [Random_0], properties {pitchMax=1.3, volumeMax=1, volumeMin=1}}
 {id=Random_0, children [WavePlayer_3,WavePlayer_4,WavePlayer_5], properties {bShouldExcludeFromBranchCulling=true, weights [1,1,1]}}
 {id=Random_1, children [WavePlayer_0,WavePlayer_1,WavePlayer_2], properties {weights [1,1,1]}}
 {
  id=WavePlayer_0
  properties {soundWaveAssetPtr=&/Knockback_Launcher_Prj_Hit_Player_Push_Back_Quad_Whsh_01}
 }
 {
  id=WavePlayer_1
  properties {soundWaveAssetPtr=&/Knockback_Launcher_Prj_Hit_Player_Push_Back_Quad_Whsh_02}
 }
 {
  id=WavePlayer_2
  properties {soundWaveAssetPtr=&/Knockback_Launcher_Prj_Hit_Player_Push_Back_Quad_Whsh_03}
 }
 {id=WavePlayer_3, properties {soundWaveAssetPtr=&/Knockback_Launcher_Prj_Hit_Player_Push_Back_Close_01}}
 {id=WavePlayer_4, properties {soundWaveAssetPtr=&/Knockback_Launcher_Prj_Hit_Player_Push_Back_Close_02}}
 {id=WavePlayer_5, properties {soundWaveAssetPtr=&/Knockback_Launcher_Prj_Hit_Player_Push_Back_Close_03}}
]
properties {
 bBypassVolumeScaleForPriority=true
 bOverrideConcurrency=true
 maxDistance=2097152
 priority=88
 soundClassObject=&Fort_WeaponRanged_Fire_3P_PSM
 volumeMultiplier=1
 attenuationOverrides {reverbWetLevelMax=0.7, reverbWetLevelMin=0.2}
 concurrencyOverrides {maxCount=1, resolutionRule=StopOldest, volumeScale=0.5}
}
