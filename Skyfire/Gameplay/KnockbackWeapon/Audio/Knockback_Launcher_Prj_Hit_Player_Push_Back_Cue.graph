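# Sound cue for a Knockback Launcher projectile hitting a player and pushing them back.
# It mixes a louder close layer with a quieter distant layer built from "Mid" waves, each picking randomly from three variations through a modulator.
# Each layer has its own attenuation and weapon-fire sound class override.
# Concurrency allows one instance, stopping the quietest, and virtualization is disabled.
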SoundCue /Skyfire/Gameplay/KnockbackWeapon/Audio/Knockback_Launcher_Prj_Hit_Player_Push_Back_Cue
root=Mixer_5
nodes [
 {
  id=Attenuation_0
  children [Modulator_1]
  properties {
   attenuationSettings=&/Knockback_Launcher_Prj_Hit_Player_Close_Att
   attenuationOverrides {falloffDistance=900, attenuationShapeExtents {x=100}}
  }
 }
 {
  id=Attenuation_4
  children [Modulator_0]
  properties {attenuationSettings=&/Knockback_Launcher_Prj_Hit_Player_Distant_Att}
 }
 {id=Mixer_0, children [SoundClass_3], properties {inputVolume [1]}}
 {id=Mixer_5, children [Mixer_0,SoundClass_0], properties {inputVolume [1.1,0.6]}}
 {id=Modulator_0, children [Random_1]}
 {id=Modulator_1, children [Random_0], properties {pitchMax=1.1, pitchMin=0.9, volumeMax=1, volumeMin=1}}
 {id=Random_0, children [WavePlayer_3,WavePlayer_4,WavePlayer_5], properties {weights [1,1,1]}}
 {id=Random_1, children [WavePlayer_6,WavePlayer_7,WavePlayer_8], properties {weights [1,1,1]}}
 {
  id=SoundClass_0
  children [Attenuation_4]
  properties {soundClassOverride=&Fort_WeaponRanged_Fire_Distant}
 }
 {
  id=SoundClass_3
  children [Attenuation_0]
  properties {soundClassOverride=&Fort_WeaponRanged_Fire_3P_PSM}
 }
 {id=WavePlayer_3, properties {soundWaveAssetPtr=&/Knockback_Launcher_Prj_Hit_Player_Push_Back_Close_01}}
 {id=WavePlayer_4, properties {soundWaveAssetPtr=&/Knockback_Launcher_Prj_Hit_Player_Push_Back_Close_02}}
 {id=WavePlayer_5, properties {soundWaveAssetPtr=&/Knockback_Launcher_Prj_Hit_Player_Push_Back_Close_03}}
 {id=WavePlayer_6, properties {soundWaveAssetPtr=&/Knockback_Launcher_Prj_Hit_Player_Push_Back_Mid_01}}
 {id=WavePlayer_7, properties {soundWaveAssetPtr=&/Knockback_Launcher_Prj_Hit_Player_Push_Back_Mid_02}}
 {id=WavePlayer_8, properties {soundWaveAssetPtr=&/Knockback_Launcher_Prj_Hit_Player_Push_Back_Mid_03}}
]
properties {
 bOverrideConcurrency=true
 maxDistance=8500
 priority=80
 virtualizationMode=Disabled
 volumeMultiplier=1
 attenuationOverrides {reverbWetLevelMax=0.7, reverbWetLevelMin=0.2}
 concurrencyOverrides {maxCount=1, resolutionRule=StopQuietest, volumeScale=0.5}
}
