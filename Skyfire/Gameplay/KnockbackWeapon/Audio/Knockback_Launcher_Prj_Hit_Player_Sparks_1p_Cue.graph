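# Sound cue for the first-person sparks layer when a Knockback Launcher projectile hits a player, judging by its name.
# Picks one of four equally weighted Backpack Surge waves through Random, then passes it through Modulator and Delay nodes at default settings.
# Overridden concurrency allows one instance, stopping the oldest, with high priority and virtualization disabled.
# It is routed to a third-person ranged weapon fire sound class despite the 1P name.
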SoundCue /Skyfire/Gameplay/KnockbackWeapon/Audio/Knockback_Launcher_Prj_Hit_Player_Sparks_1p_Cue.Knockback_Launcher_Prj_Hit_Player_Sparks_1P_Cue
root=Delay_2
nodes [
 {id=Delay_2, children [Modulator_2]}
 {id=Modulator_2, children [Random_2]}
 {id=Random_2, children [WavePlayer_6,WavePlayer_7,WavePlayer_8,WavePlayer_9], properties {weights [1,1,1,1]}}
 {id=WavePlayer_6, properties {soundWaveAssetPtr=&/Knockback_Launcher_Backpack_Surge_04}}
 {id=WavePlayer_7, properties {soundWaveAssetPtr=&/Knockback_Launcher_Backpack_Surge_01}}
 {id=WavePlayer_8, properties {soundWaveAssetPtr=&/Knockback_Launcher_Backpack_Surge_02}}
 {id=WavePlayer_9, properties {soundWaveAssetPtr=&/Knockback_Launcher_Backpack_Surge_03}}
]
properties {
 bBypassVolumeScaleForPriority=true
 bOverrideConcurrency=true
 maxDistance=2097152
 priority=88
 soundClassObject=&Fort_WeaponRanged_Fire_3P_PSM
 virtualizationMode=Disabled
 volumeMultiplier=0.7
 attenuationOverrides {reverbWetLevelMax=0.7, reverbWetLevelMin=0.2}
 concurrencyOverrides {maxCount=1, resolutionRule=StopOldest, volumeScale=0.5}
}
