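# Sound cue for the sparks sound when a Knockback Launcher projectile hits a player, judging by its name.
# The first two quality-level slots share one chain of delay, modulator and equal-weight random pick among four backpack surge waves; the other two slots are empty.
# It uses shared Skyfire backpack attenuation and overrides concurrency to one instance, stopping the quietest.
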SoundCue /Skyfire/Gameplay/KnockbackWeapon/Audio/Knockback_Launcher_Prj_Hit_Player_Sparks_Cue
root=QualityLevel_0
nodes [
 {id=Delay_2, children [Modulator_2]}
 {id=Modulator_2, children [Random_2]}
 {id=QualityLevel_0, children [Delay_2,Delay_2,null,null]}
 {id=Random_2, children [WavePlayer_0,WavePlayer_1,WavePlayer_2,WavePlayer_9], properties {weights [1,1,1,1]}}
 {id=WavePlayer_0, properties {soundWaveAssetPtr=&/Knockback_Launcher_Backpack_Surge_04}}
 {id=WavePlayer_1, properties {soundWaveAssetPtr=&/Knockback_Launcher_Backpack_Surge_01}}
 {id=WavePlayer_2, properties {soundWaveAssetPtr=&/Knockback_Launcher_Backpack_Surge_02}}
 {id=WavePlayer_9, properties {soundWaveAssetPtr=&/Knockback_Launcher_Backpack_Surge_03}}
]
properties {
 attenuationSettings=/Skyfire/Sound/Skyfire_Backpack_Att
 bOverrideConcurrency=true
 maxDistance=1350
 priority=30
 soundClassObject=&Fort_WeaponRanged_Fire_3P_PSM
 volumeMultiplier=0.7
 attenuationOverrides {reverbWetLevelMax=0.7, reverbWetLevelMin=0.2}
 concurrencyOverrides {maxCount=1, resolutionRule=StopQuietest, volumeScale=0.5}
}
