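# Sound cue for the UI hit notification when a knockback launcher projectile hits a player, routed to the instigator 2D hit-notification sound class.
# A single wave passes through a Modulator that randomly raises pitch, then into a Quality Level node with only its first two slots connected.
# Overridden concurrency allows one instance per owner, stopping the oldest on retrigger.
# Virtualization is disabled, and volume is boosted at zero priority with volume-based priority scaling bypassed.
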SoundCue /Skyfire/Gameplay/KnockbackWeapon/Audio/Knockback_Launcher_Prj_Hit_Player_UI_Cue
root=QualityLevel_0
nodes [
 {id=Modulator_0, children [WavePlayer_0], properties {pitchMax=1.2, pitchMin=1, volumeMax=1, volumeMin=1}}
 {id=QualityLevel_0, children [Modulator_0,Modulator_0,null,null]}
 {id=WavePlayer_0, properties {soundWaveAssetPtr=&/Knockback_Launcher_Prj_Hit_Player_UI_01}}
]
properties {
 bBypassVolumeScaleForPriority=true
 bOverrideConcurrency=true
 maxDistance=2097152
 priority=0
 soundClassObject=&Fort_2DHitNotifications_Instigate_PSM
 virtualizationMode=Disabled
 volumeMultiplier=1.5
 concurrencyOverrides {bLimitToOwner=true, maxCount=1, resolutionRule=StopOldest}
}
