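# Looping sound cue for the Knockback Launcher projectile while it is in the air.
# A looping wave gets random pitch variation, then an envelope that quickly fades volume in while pitch falls from high to slightly below normal.
# A smoothed Doppler node follows, feeding the first two quality levels; the other two are empty.
# It uses projectile attenuation, disables virtualization and overrides concurrency to one instance, stopping the quietest.
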SoundCue /Skyfire/Gameplay/KnockbackWeapon/Audio/Knockback_Launcher_Prj_InAir_Loop_Cue
root=QualityLevel_0
nodes [
 {id=Doppler_0, children [Enveloper_0], properties {bUseSmoothing=true, dopplerIntensity=1.2, smoothingInterpSpeed=0.75}}
 {
  id=Enveloper_0
  children [Modulator_0]
  properties {
   pitchCurve {editorCurveData {keys [<at 0 2>,<at 3 0.95>]}}
   volumeCurve {editorCurveData {keys [{arriveTangent=4.8462, interpMode=RCIM_Cubic, leaveTangent=4.8462, tangentMode=RCTM_User, time=0, value=0}, <at 0.25 1>]}}
  }
 }
 {id=Modulator_0, children [WavePlayer_4], properties {pitchMax=1.1, pitchMin=0.9, volumeMax=1, volumeMin=1}}
 {id=QualityLevel_0, children [Doppler_0,Doppler_0,null,null]}
 {id=WavePlayer_4, properties {bLooping=true, soundWaveAssetPtr=&/Knockback_Launcher_Prj_InAir_Loop}}
]
properties {
 attenuationSettings=&/Knockback_Launcher_Prj_InAir_Att
 bOverrideConcurrency=true
 maxDistance=17000
 priority=0
 soundClassObject=&Fort_WeaponRanged_Fire_3P_PSM
 virtualizationMode=Disabled
 volumeMultiplier=2
 attenuationOverrides {falloffDistance=5000, attenuationShapeExtents {x=1000}}
 concurrencyOverrides {bVolumeScaleCanRelease=true, maxCount=1, resolutionRule=StopQuietest, volumeScale=0.2, volumeScaleAttackTime=0.2, volumeScaleReleaseTime=2}
 concurrencySet [/Game/Sounds/Concurrency_Rules/Athena_Explosive_Fuse_Conc]
}
