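# Sound cue for the distant whip sound of the Knockback Launcher projectile, classed as a ranged-weapon projectile effect.
# A quality switch feeds its first two levels through a modulator into an equally weighted random pick of four wave variations, leaving the other two levels empty.
# It references a dedicated whip attenuation asset, disables virtualization, and overrides concurrency to cap simultaneous instances by stopping the oldest.
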SoundCue /Skyfire/Gameplay/KnockbackWeapon/Audio/Knockback_Launcher_Prj_Whip_Distant_Cue
root=QualityLevel_0
nodes [
 {id=Modulator_0, children [Random_0]}
 {id=QualityLevel_0, children [Modulator_0,Modulator_0,null,null]}
 {
  id=Random_0
  children [WavePlayer_0,WavePlayer_1,WavePlayer_2,WavePlayer_7]
  properties {bShouldExcludeFromBranchCulling=true, weights [1,1,1,1]}
 }
 {id=WavePlayer_0, properties {soundWaveAssetPtr=&/Knockback_Launcher_Prj_Whip_Distant_01}}
 {id=WavePlayer_1, properties {soundWaveAssetPtr=&/Knockback_Launcher_Prj_Whip_Distant_02}}
 {id=WavePlayer_2, properties {soundWaveAssetPtr=&/Knockback_Launcher_Prj_Whip_Distant_03}}
 {id=WavePlayer_7, properties {soundWaveAssetPtr=&/Knockback_Launcher_Prj_Whip_Distant_04}}
]
properties {
 attenuationSettings=&/Knockback_Launcher_Prj_Whip_Att
 bOverrideConcurrency=true
 maxDistance=23000
 priority=40
 soundClassObject=&Fort_WeaponRanged_Projectile_Effects
 virtualizationMode=Disabled
 volumeMultiplier=2
 attenuationOverrides {
  bEnableReverbSend=false
  bUseComplexCollisionForOcclusion=true
  falloffDistance=1000
  occlusionVolumeAttenuation=0
  stereoSpread=0
  attenuationShapeExtents {x=512}
 }
 concurrencyOverrides {maxCount=2, resolutionRule=StopOldest, volumeScaleAttackTime=0}
}
