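# Sound cue for the very close whip-by of a Knockback Launcher projectile, using the shared whip attenuation asset.
# A quality switch feeds a Modulator over an equally weighted random pick of four "Close" whip waves in its first two slots, leaving the other two empty.
# The cue raises pitch and volume, disables virtualization, and limits concurrency to one instance, stopping the oldest.
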SoundCue /Skyfire/Gameplay/KnockbackWeapon/Audio/Knockback_Launcher_Prj_Whip_Very_Close_Cue
root=QualityLevel_0
nodes [
 {id=Modulator_0, children [Random_1]}
 {id=QualityLevel_0, children [Modulator_0,Modulator_0,null,null]}
 {
  id=Random_1
  children [WavePlayer_3,WavePlayer_4,WavePlayer_5,WavePlayer_6]
  properties {bShouldExcludeFromBranchCulling=true, weights [1,1,1,1]}
 }
 {id=WavePlayer_3, properties {soundWaveAssetPtr=&/Knockback_Launcher_Prj_Whip_Close_01}}
 {id=WavePlayer_4, properties {soundWaveAssetPtr=&/Knockback_Launcher_Prj_Whip_Close_02}}
 {id=WavePlayer_5, properties {soundWaveAssetPtr=&/Knockback_Launcher_Prj_Whip_Close_03}}
 {id=WavePlayer_6, properties {soundWaveAssetPtr=&/Knockback_Launcher_Prj_Whip_Close_04}}
]
properties {
 attenuationSettings=&/Knockback_Launcher_Prj_Whip_Att
 bOverrideConcurrency=true
 maxDistance=23000
 pitchMultiplier=3
 priority=0
 soundClassObject=&Fort_WeaponRanged_Projectile_Effects
 virtualizationMode=Disabled
 volumeMultiplier=2
 attenuationOverrides {
  bAttenuate=false
  bEnableReverbSend=false
  bSpatialize=false
  bUseComplexCollisionForOcclusion=true
  falloffDistance=1000
  occlusionVolumeAttenuation=0
  stereoSpread=0
  attenuationShapeExtents {x=512}
 }
 concurrencyOverrides {maxCount=1, resolutionRule=StopOldest}
}
