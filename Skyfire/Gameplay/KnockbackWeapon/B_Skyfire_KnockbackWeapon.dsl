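; Skyfire knockback weapon Blueprint that adds rarity-colored emissives and overheat visuals and audio.
; While in play, it listens to the instigator's SkyfirePlayerPawnComponent for earned loot rarity changes and recolors the weapon mesh from the rarity palette.
; Firing starts a looping timer that scales a dynamic muzzle-glow material's opacity with heat until it reaches zero; local unequip resets both.
; Overheating triggers an effect and sound.
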(event EventBeginPlay
  (bind _returnvalue (Game|GetInstigator))
  (bind _returnvalue_2 (Actor|GetComponentbyClass _returnvalue "/Script/SkyfireRuntime.SkyfirePlayerPawnComponent"))
  ;[1 0.2311 0.9279] Add Bind event for picking up loot orbs
  (|CallParentFunction)
  (Utilities|IsValid _returnvalue
    (:Is%20Valid
      (Utilities|IsValid _returnvalue_2
        (:Is%20Valid
          (EventDispatchers|BindEventtoOnEarnedLootRarityChanged _returnvalue_2 (EventDispatchers|CreateEvent :!event Custom|ChangeWeaponEmissive)))))))

(event EventEndPlay
  (params (EndPlayReason :type EEndPlayReason%20Enum :default "Destroyed"))
  (bind _returnvalue (Game|GetInstigator))
  (bind _returnvalue_2 (Actor|GetComponentbyClass _returnvalue "/Script/SkyfireRuntime.SkyfirePlayerPawnComponent"))
  ;[1 0.2311 0.9279] Add Bind event for picking up loot orbs
  (|CallParentFunction EndPlayReason)
  (Utilities|IsValid _returnvalue
    (:Is%20Valid
      (Utilities|IsValid _returnvalue_2
        (:Is%20Valid
          (EventDispatchers|UnbindEventfromOnEarnedLootRarityChanged _returnvalue_2 (EventDispatchers|CreateEvent :!event Custom|ChangeWeaponEmissive)))))))

(event Custom|ChangeWeaponEmissive
  (params (NewRarity :type EFortRarity%20Enum :default "Common"))
  (bind _weaponmesh (Variables|Visual|GetWeaponMesh))
  ;[1 0 0.8191] Update Weapon Emissives when loot orb is picked up
  (bind _returnvalue (Fortnite|GetRarityColorPalette NewRarity))
  (Utilities|IsValid _weaponmesh
    (:Is%20Valid
      (Rendering|Material|SetVectorParameterValueonMaterials _weaponmesh (Variables|Default|GetEmissiveMultiplierParameter) (Math|Conversions|ToVector(LinearColor) (Utilities|Struct|BreakFortColorPalette _returnvalue))))))

(event Overheating|EventOnStartOverheated
  ;[0.9386 0 1] Activate overheat FX
  (|CallParentFunction)
  (Components|Activation|Activate (Variables|Default|GetOverheatFX) false)
  (Audio|Components|Audio|Play (Variables|Default|GetOnOverheatSound) 0.0))

(event EventOnPlayWeaponFireFX
  (params (bPersistentFire :type Boolean) (bSecondaryFire :type Boolean))
  ;[0.1415 0.2219 1] On fire, start timer to handle heat FX.
  (|CallParentFunction bPersistentFire bSecondaryFire)
  (bind _returnvalue (Utilities|Time|SetTimerbyEvent (EventDispatchers|CreateEvent :!event Custom|UpdateOverheatFX) 0.0333 true))
  (Variables|Default|SetOverheatFXTimer _returnvalue))

(event Custom|UpdateOverheatFX
  (bind _returnvalue (Overheat|GetCurrentOverheatValue))
  ;[0.07926 1 0.6636] Adjust the overheat FX based on the current overheat values.
  (Networking|CanExecuteCosmeticEvents
    (:True
      (if (> _returnvalue 0.0)
        (Rendering|Material|SetScalarParameterValue (Variables|Default|GetMuzzleGlowMID) (Variables|Default|GetOpacityParam) (Math|Float|Lerp 0.0 1.0 (/ _returnvalue (Math|Float|Max(Float) (Overheat|GetOverheatingMaxValue) 1.0))) :!class /Script/Engine.MaterialInstanceDynamic)
        (else
          (Utilities|Time|ClearandInvalidateTimerbyHandle (Variables|Default|GetOverheatFXTimer)) ; Clear timer when visuals return to default
        )))))

(event FortWeapon|EventOnLocalUnEquip
  ;[0.9716 1 0.1146] Clear timer and reset overheat FX when swapped.
  (|CallParentFunction)
  (Utilities|Time|ClearandInvalidateTimerbyHandle (Variables|Default|GetOverheatFXTimer))
  (Networking|CanExecuteCosmeticEvents
    (:True
      (Rendering|Material|SetScalarParameterValue (Variables|Default|GetMuzzleGlowMID) (Variables|Default|GetOpacityParam) 0.0 :!class /Script/Engine.MaterialInstanceDynamic))))

(fn ConstructionScript
  ;[0.0603 1 0.08222] Create dynamic material so we can adjust overheat visuals.
  (Networking|CanExecuteCosmeticEvents
    (:True
      (bind _returnvalue (Rendering|Material|CreateDynamicMaterialInstance (Variables|Default|GetOverheatMesh) 0 0 "None" :!class /Script/Engine.PrimitiveComponent))
      (Variables|Default|SetMuzzleGlowMID _returnvalue))
    (:False
      _))
  (|CallParentFunction))
