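# Unlit, two-sided translucent Niagara particle material for knockback weapon radial effects.
# Builds radial-style UVs from center distance and an arctangent (fed G on both axes), offset and tiled by dynamic parameters, with optional swizzle and flips.
# Base and optional secondary sprites, multiplied by a radial mask and particle alpha, drive depth-faded opacity.
# Noise blends two colors into emissive, optionally adding the mask and particle tint; a color panner sits unused.
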Material /Skyfire/Gameplay/KnockbackWeapon/FX/M_KnockbackWeapon_Radial
graph {
 properties {
  bUsedWithNiagaraMeshParticles=true
  bUsedWithNiagaraSprites=true
  blendMode=BLEND_Translucent
  pixelDepthOffsetMode=PDOM_Legacy
  shadingModel=MSM_Unlit
  twoSided=true
  usedShadingModels=Unlit
 }
 nodes [
  Constant() {r=1.75}
  Constant() {r=1}
  Constant() {r=3}
  n1 = ParticleColor()
  n2 = DynamicParameter() {parameterIndex=2, paramNames [Base Sprite Tile U, Base Sprite Tile V, Radial Mask Tile U, Radial Mask Tile V]}
  n3 = DynamicParameter() {
   parameterIndex=1
   defaultValue=<rgba 1 1 1 2>
   paramNames [Base Sprite Offset U, Base Sprite Offset V, Radial Mask Offset U, Radial Mask Offset V]
  }
  n4 = ComponentMask(None=@n3.Base Sprite Offset U) {r=true}
  n5 = ComponentMask(None=@n3.Base Sprite Offset V) {g=true}
  n6 = Subtract(A=TextureCoordinate()) {constB=0.5}
  n7 = MaterialFunctionCall(Float2=@n6) {materialFunction=&BreakOutFloat2Components}
  n8 = Arctangent2Fast(Y=@n7.G, X=@n7.G)
  n9 = MaterialFunctionCall() {materialFunction=&Pi}
  n10 = Constant() {r=2}
  n11 = AppendVector(
   A=SquareRoot(None=MaterialFunctionCall(Vector 2=@n6) {materialFunction=/Engine/Functions/Engine_MaterialFunctions02/Utility/VectorLength}.V2 Length)
   B=Divide(
    A=Add(A=@n8, B=@n9.Result)
    B=MaterialFunctionCall(Multiplier=@n10) {materialFunction=&Pi}.Result
   )
  )
  n12 = Constant2Vector() {g=1, r=1}
  n13 = ComponentMask(None=@n12) {r=true}
  n14 = DynamicParameter() {defaultValue=<rgba 1 1 1 2>, paramNames [Sprite Extrusion Speed, Sprite Spin Speed, Radial Mask Strength, Radial Mask Power]}
  n15 = ComponentMask(None=@n12) {g=true}
  n16 = Add(
   A=AppendVector(A=Multiply(A=@n13, B=@n14.Sprite Extrusion Speed), B=Multiply(A=@n15, B=@n14.Sprite Spin Speed))
   B=Add(A=AppendVector(A=@n4, B=@n5), B=@n11)
  )
  n17 = MaterialFunctionCall(XY=@n16) {materialFunction=&Swizzle}
  n18 = ScalarParameter() {group="1) Base Sprite Rotation", parameterName=Base Sprite - Rotate UVs, sortPriority=0}
  n19 = Constant() {r=1}
  n20 = Constant() {r=-1}
  n21 = LinearInterpolate(A=@n19, B=@n20, Alpha=ScalarParameter() {group="1) Base Sprite Rotation", parameterName=Base Sprite - Flip UVs Horizontal, sortPriority=1})
  n22 = LinearInterpolate(A=@n19, B=@n20, Alpha=ScalarParameter() {group="1) Base Sprite Rotation", parameterName=Base Sprite - Flip UVs Vertically, sortPriority=2})
  n23 = Multiply(
   A=AppendVector(A=@n2.Base Sprite Tile U, B=@n2.Base Sprite Tile U)
   B=Multiply(A=LinearInterpolate(A=@n16, B=@n17.YX, Alpha=@n18), B=AppendVector(A=@n21, B=@n22))
  )
  n24 = TextureCoordinate()
  n25 = DDX(Value=@n24)
  n26 = DDY(Value=@n24)
  n27 = TextureSampleParameter2D!(UVs=@n23, "DDX(UVs)"=@n25, "DDY(UVs)"=@n26, Apply View MipBias) {
   constMipValue=0
   group=Base Sprite Texture
   mipValueMode=TMVM_Derivative
   parameterName=Base Sprite
   sortPriority=60
   texture=&T_ShieldCaustics
  }
  n28 = TextureSampleParameter2D(UVs=@n23, "DDX(UVs)"=@n25, "DDY(UVs)"=@n26) {
   constMipValue=0
   group=Base Sprite Texture
   mipValueMode=TMVM_Derivative
   parameterName=Secondary Sprite
   samplerType=SAMPLERTYPE_LinearColor
   sortPriority=60
   texture=/Game/Characters/Player/Male/Medium/Bodies/M_MED_SRDrift/Textures/T_ElectricNoise
  }
  n29 = LinearInterpolate(A=@n27.R, B=Add(A=@n28.R, B=@n27.R), Alpha=ScalarParameter() {parameterName=Add Secondary Sprite?})
  n30 = DynamicParameter() {parameterIndex=2, paramNames [Base Sprite Tile U, Base Sprite Tile V, Radial Mask Tile U, Radial Mask Tile V]}
  n31 = Add(A=AppendVector(A=ComponentMask(None=@n3.Radial Mask Offset U) {r=true}, B=ComponentMask(None=@n3.Radial Mask Offset V) {g=true}), B=@n11)
  n32 = LinearInterpolate(
   A=@n31
   B=MaterialFunctionCall(XY=@n31) {materialFunction=&Swizzle}.YX
   Alpha=ScalarParameter() {defaultValue=1, group="1.1) Radial Mask Rotation", parameterName=Radial Mask - Rotate UVs, sortPriority=3}
  )
  n33 = Constant() {r=1}
  n34 = Constant() {r=-1}
  n35 = AppendVector(
   A=LinearInterpolate(A=@n33, B=@n34, Alpha=ScalarParameter() {group="1.1) Radial Mask Rotation", parameterName=Radial Mask - Flip UVs Horizontal, sortPriority=4})
   B=LinearInterpolate(A=@n33, B=@n34, Alpha=ScalarParameter() {group="1.1) Radial Mask Rotation", parameterName=Radial Mask - Flip UVs Vertically, sortPriority=5})
  )
  n36 = TextureSampleParameter2D(
   UVs=Multiply(A=AppendVector(A=@n30.Radial Mask Tile U, B=@n30.Radial Mask Tile U), B=Multiply(A=@n32, B=@n35))
   "DDX(UVs)"=@n25
   "DDY(UVs)"=@n26
  ) {
   constMipValue=0
   group=Radial Mask Texture
   mipValueMode=TMVM_Derivative
   parameterName=Radial Mask
   sortPriority=70
   texture=/Game/Effects/Fort_Effects/Textures/Ramps/T_Ramp_Smooth_Reverse
  }
  n37 = DynamicParameter() {defaultValue=<rgba 1 1 1 2>, paramNames [Sprite Extrusion Speed, Sprite Spin Speed, Radial Mask Strength, Radial Mask Power]}
  n38 = Clamp(None=Multiply(A=Power(Base=@n36.RGB, Exp=@n37.Radial Mask Power), B=@n37.Radial Mask Strength))
  DepthFade(
   Opacity=Multiply(A=@n1.A, B=Multiply(A=@n29, B=@n38))
   FadeDistance=DynamicParameter() {parameterIndex=3, defaultValue=<rgba 1 1 1 2>, paramNames [Depth Fade, N/A, N/A, N/A]}.Depth Fade
  )
  n39 = TextureSampleParameter2D(UVs=@n23, "DDX(UVs)"=@n25, "DDY(UVs)"=@n26) {
   group=Color
   mipValueMode=TMVM_Derivative
   parameterName=Color Noise
   sortPriority=0
   texture=&T_Inky_Smoke_Tile
  }
  n40 = ScalarParameter() {defaultValue=0.5, group=Color, parameterName=ColorNoise_Power, sortPriority=7}
  n41 = LinearInterpolate(
   A=VectorParameter() {group=Color, parameterName=Color A, sortPriority=1, defaultValue=<rgba 0.491 0.0887 0.7529 1>}.RGB
   B=VectorParameter() {group=Color, parameterName=Color B, sortPriority=2, defaultValue=<rgba 0.0222 0.0666 0.4179 1>}.RGB
   Alpha=Saturate(None=Power(Base=@n39.R, Exp=@n40) {constExponent=0.5})
  )
  n42 = Multiply(A=MaterialFunctionCall(LDR Input=Saturate(None=@n41), HDR Intensity Multiplier=Constant() {r=1}) {materialFunction=&DeriveHDRfromLDR}.RemappedRange)
  n43 = Multiply(A=@n36.RGB, B=Constant() {r=1}) {constB=1.75}
  n44 = Constant() {r=1}
  n45 = LinearInterpolate(B=@n1.RGB, Alpha=Saturate(None=Power(Base=@n43, Exp=@n44) {constExponent=3}))
  n46 = Clamp(None=LinearInterpolate(A=@n42, B=Add(A=@n42, B=@n45), Alpha=ScalarParameter() {group=Color, parameterName=Add Radial Mask to Color?}))
  n47 = ParticleColor()
  MaterialFunctionCall(Emissive=LinearInterpolate(A=@n46, B=Multiply(A=@n47.RGB, B=@n46), Alpha=ScalarParameter() {group=Color, parameterName=Multiply Particle Color?})) {materialFunction=&MF_FX_EmissiveAdjustments}
  n48 = ScalarParameter() {defaultValue=3, group=Color, parameterName=Color_XTile, sortPriority=3}
  n49 = ScalarParameter() {defaultValue=3, group=Color, parameterName=Color_YTile, sortPriority=4}
  Panner(
   Coordinate=Multiply(A=TextureCoordinate(), B=AppendVector(A=@n48, B=@n49))
   Time=Time()
   Speed=AppendVector(
    A=ScalarParameter() {defaultValue=-0.1, group=Color, parameterName=Color_XSpeed, sortPriority=5}
    B=ScalarParameter() {defaultValue=-0.2, group=Color, parameterName=Color_YSpeed, sortPriority=6}
   )
  )
  Power()
 ]
}
