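# Unlit material for the Skyfire knockback weapon projectile, usable on Niagara mesh particles, rendering an animated purple-blue energy swirl.
# Panning smoke noise blends two swirl colors, composited over a dark background by a mask subtracting scrolling noise from turbulence.
# A shared panning distortion texture warps both mask textures, and a UV.y gradient fades the mask.
# Swirl-tinted inverted Fresnel is added, then time-of-day brightness scaling feeds the FX emissive adjustment function.
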Material /Skyfire/Gameplay/KnockbackWeapon/FX/M_Skyfire_KnockbackWeapon_Projectile
graph {
 properties {bUsedWithNiagaraMeshParticles=true, pixelDepthOffsetMode=PDOM_Legacy, shadingModel=MSM_Unlit, usedShadingModels=Unlit}
 nodes [
  Constant()
  n1 = VectorParameter() {group=Color, parameterName=Background Color, sortPriority=3, defaultValue=<rgba 0.0183 0.0034 0.0521 1>}
  n2 = VectorParameter() {group=Color, parameterName=Swirl Color A, sortPriority=1, defaultValue=<rgba 0.491 0.0887 0.7529 1>}
  n3 = VectorParameter() {group=Color, parameterName=Swirl Color B, sortPriority=2, defaultValue=<rgba 0.0222 0.0666 0.4179 1>}
  n4 = TextureCoordinate()
  n5 = AppendVector(
   A=ScalarParameter() {defaultValue=3, group=Color, parameterName=Color_XTile, sortPriority=3}
   B=ScalarParameter() {defaultValue=3, group=Color, parameterName=Color_YTile, sortPriority=4}
  )
  n6 = ScalarParameter() {defaultValue=-0.1, group=Color, parameterName=Color_XSpeed, sortPriority=5}
  n7 = ScalarParameter() {defaultValue=-0.2, group=Color, parameterName=Color_YSpeed, sortPriority=6}
  n8 = TextureSampleParameter2D(UVs=Panner(Coordinate=Multiply(A=@n4, B=@n5), Time=Time(), Speed=AppendVector(A=@n6, B=@n7))) {group=Color, parameterName=Color Noise, sortPriority=0, texture=&T_Inky_Smoke_Tile}
  n9 = ScalarParameter() {defaultValue=0.5, group=Color, parameterName=ColorNoise_Power, sortPriority=7}
  n10 = LinearInterpolate(A=@n2.RGB, B=@n3.RGB, Alpha=Saturate(None=Power(Base=@n8.R, Exp=@n9) {constExponent=0.5}))
  n11 = Saturate(None=@n10)
  n12 = Multiply(
   A=TextureCoordinate()
   B=AppendVector(
    A=ScalarParameter() {defaultValue=3, group=Emissive, parameterName=TextureB_XTile, sortPriority=5}
    B=ScalarParameter() {defaultValue=3, group=Emissive, parameterName=TextureB_YTile, sortPriority=6}
   )
  )
  n13 = TextureCoordinate()
  n14 = AppendVector(
   A=ScalarParameter() {defaultValue=3, group=Distortion, parameterName=Distortion_XTile, sortPriority=1}
   B=ScalarParameter() {defaultValue=3, group=Distortion, parameterName=Distortion_YTile, sortPriority=2}
  )
  n15 = ScalarParameter() {defaultValue=0.1, group=Distortion, parameterName=Distortion_XSpeed, sortPriority=3}
  n16 = ScalarParameter() {defaultValue=0.2, group=Distortion, parameterName=Distortion_YSpeed, sortPriority=4}
  n17 = TextureSampleParameter2D(UVs=Panner(Coordinate=Multiply(A=@n13, B=@n14), Time=Time(), Speed=AppendVector(A=@n15, B=@n16))) {
   group=Distortion
   parameterName=Distortion Texture
   samplerType=SAMPLERTYPE_LinearGrayscale
   sortPriority=0
   texture=/Game/Global/Textures/Default/Flowmaps/T_StormSwirlMask
  }
  n18 = Multiply(A=@n17.R, B=ScalarParameter() {defaultValue=0.05, group=Distortion, parameterName=Distortion_Amount, sortPriority=5})
  n19 = AppendVector(A=@n18, B=@n18)
  n20 = ScalarParameter() {group=Emissive, parameterName=TextureB_XSpeed, sortPriority=7}
  n21 = ScalarParameter() {defaultValue=0.2, group=Emissive, parameterName=TextureB_YSpeed, sortPriority=8}
  n22 = TextureSampleParameter2D(UVs=Panner(Coordinate=Add(A=@n12, B=@n19), Time=Time(), Speed=AppendVector(A=@n20, B=@n21))) {group=Emissive, parameterName=Texture B, sortPriority=5, texture=/Game/Global/Textures/Default/Caustics/T_Turbulence_Seamless}
  n23 = Multiply(A=Add(A=@n22.R) {constB=0.15}, B=ScalarParameter() {defaultValue=1.25, group=Emissive, parameterName=TextureB_Intensity, sortPriority=9}) {constB=1.25}
  n24 = ScalarParameter() {defaultValue=4, group=Emissive, parameterName=TextureA_XTile, sortPriority=1}
  n25 = ScalarParameter() {defaultValue=4, group=Emissive, parameterName=TextureA_YTile, sortPriority=2}
  n26 = Add(A=Multiply(A=TextureCoordinate(), B=AppendVector(A=@n24, B=@n25)), B=@n19)
  n27 = Time()
  n28 = AppendVector(
   A=ScalarParameter() {group=Emissive, parameterName=TextureA_XSpeed, sortPriority=3}
   B=ScalarParameter() {defaultValue=0.2, group=Emissive, parameterName=TextureA_YSpeed, sortPriority=4}
  )
  n29 = Multiply(
   A=TextureSampleParameter2D(UVs=Panner(Coordinate=@n26, Time=@n27, Speed=@n28)) {group=Emissive, parameterName=Texture A, sortPriority=0, texture=/Game/Global/Textures/Default/Noise/T_Tiling_Noise_01}.R
   B=ScalarParameter() {defaultValue=0.01, group=Emissive, parameterName=TextureA_Intensity, sortPriority=3}
  ) {constB=0.01}
  n30 = TextureCoordinate()
  n31 = Multiply(A=Power(Base=ComponentMask(None=@n30) {g=true})) {constB=10}
  n32 = Saturate(None=Multiply(A=Subtract(A=@n23, B=@n29), B=Saturate(None=@n31)))
  n33 = MaterialFunctionCall(
   Brightness=ScalarParameter() {defaultValue=1.2, group=Fresnel, parameterName=Fresnel_Brightness, sortPriority=0}
   Exponent=ScalarParameter() {defaultValue=10, group=Fresnel, parameterName=Fresnel_Exponent, sortPriority=1}
  ) {materialFunction=&MF_Fresnel_FX}
  n34 = ScalarParameter() {defaultValue=1, group=Fresnel, parameterName=Fresnel Intensity, sortPriority=2}
  n35 = Add(
   A=LinearInterpolate(
    A=@n1.RGB
    B=MaterialFunctionCall(LDR Input=@n11) {materialFunction=&DeriveHDRfromLDR}.RemappedRange
    Alpha=MaterialFunctionCall(LDR Input=@n32) {materialFunction=&DeriveHDRfromLDR}.RemappedRange
   )
   B=Multiply(A=@n10, B=Multiply(A=@n33.Fresnel Inverted, B=@n34) {constB=5})
  )
  n36 = ScalarParameter() {defaultValue=1, group=TOD, parameterName=TOD_Morning, sortPriority=0}
  n37 = ScalarParameter() {defaultValue=1, group=TOD, parameterName=TOD_Day, sortPriority=1}
  n38 = ScalarParameter() {defaultValue=1, group=TOD, parameterName=TOD_Evening, sortPriority=2}
  n39 = ScalarParameter() {defaultValue=1, group=TOD, parameterName=TOD_Night, sortPriority=3}
  MaterialFunctionCall(Emissive=Multiply(
    A=Saturate(None=@n35)
    B=MaterialFunctionCall(Value_Morning=@n36, Value_Day=@n37, Value_Evening=@n38, Value_Night=@n39) {materialFunction=&Scalar_Time_of_Day}.Result
   )) {materialFunction=&MF_FX_EmissiveAdjustments}
  OneMinus(None=@n31)
 ]
}
