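# Niagara projectile effect for the Skyfire knockback weapon, sized by User.ProjectileScale and tinted by User.Color_01.
# Halo and rotating Streaks sprites form the head, with Streaks alpha varying by time of day.
# MeshTrail adds a long pulsing, two-renderer mesh, while Rings continuously spawns short-fading, expanding mesh rings that brighten with User.ADS.
# Ribbon trails, cones, Mesh and OverAirglow emitters are disabled; User.RingSize feeds only a disabled module.
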NiagaraSystem /Skyfire/Gameplay/KnockbackWeapon/FX/NS_Skyfire_KnockbackWeapon_Proj
dependencies {
 usedDataInterfaces [&NiagaraDataInterfaceCurve, &NiagaraDataInterfaceVectorCurve]
 usedDynamicInputs []
 usedModules [&SystemState]
}
emitters {
 Halo {
  emitterUpdateScript [
   &EmitterState(Life Cycle Mode=::NewEnumerator1, Loop Behavior=::NewEnumerator1, Loop Duration=2, UseLoopDelay=false)
   &SpawnBurst_Instantaneous(Spawn Count=1, Use Spawn Probability=false)
  ]
  particleSpawnScript [&InitializeParticle(
    Color=$User.Color_01
    Sprite Size Mode=::NewEnumerator3
    Uniform Sprite Size=&Multiply_Float(A=$User.ProjectileScale, B=275)
    UsePositionOffset=false
    Write Material Random=true
   )]
  particleUpdateScript [
   &ParticleState(Kill Particles When Lifetime Has Elapsed=false)
   &DynamicMaterialParameters(
    Index 0 Param 1=0.7
    Index 0 Param 2=0.25
    Index 0 Param 3=1
    Module.Param0WriteEnabled=true
    Module.Param1WriteEnabled=true
    Module.Param2WriteEnabled=true
    Module.Param3WriteEnabled=true
   )
   &ScaleColor(Scale Alpha=0.8, ScaleA=true, ScaleRGB=true)
  ]
  rendererClasses [&NiagaraSpriteRendererProperties]
  renderers [{rendererClass=&NiagaraSpriteRendererProperties}]
 }
 Mesh {
  bEnabled=false
  emitterUpdateScript [
   &EmitterState(Life Cycle Mode=::NewEnumerator1, Loop Behavior=::NewEnumerator1, Loop Duration=2, UseLoopDelay=false)
   &SpawnBurst_Instantaneous(Spawn Count=1, Use Spawn Probability=false)
  ]
  particleSpawnScript [
   &InitializeParticle(
    Color=<rgba 0.8603 0.3698 1 1>
    Mesh Scale Mode=::NewEnumerator3
    Mesh Uniform Scale=&Multiply_Float(A=$User.ProjectileScale, B=1)
    Sprite Size Mode=::NewEnumerator3
    Uniform Sprite Size=&Multiply_Float(A=$User.ProjectileScale, B=250)
    UsePositionOffset=false
    Write Material Random=true
   )
   &InitialMeshOrientation(Orientation Vector=<0 0 -1>, Use Orientation Vector=true, Use Rotation Vector=false)
  ]
  particleUpdateScript [
   &ParticleState(Kill Particles When Lifetime Has Elapsed=false)
   &DynamicMaterialParameters(
    Index 0 Param 1=3
    Index 0 Param 2=1
    Module.Param0WriteEnabled=true
    Module.Param1WriteEnabled=true
    Module.Param2WriteEnabled=true
    Module.Param3WriteEnabled=true
   )
   &ScaleColor(Scale Alpha=0.5, ScaleA=true, ScaleRGB=true)
  ]
  rendererClasses [&NiagaraSpriteRendererProperties, &NiagaraMeshRendererProperties]
  renderers [
   {rendererClass=&NiagaraSpriteRendererProperties}
   {rendererClass=&NiagaraMeshRendererProperties, rendererIndex=1}
  ]
 }
 MeshTrail {
  emitterUpdateScript [
   &EmitterState()
   &SpawnBurst_Instantaneous(Spawn Count=1, Use Spawn Probability=false)
  ]
  particleSpawnScript [
   &InitializeParticle(
    Color=$User.Color_01
    Lifetime=3
    Mesh Scale=&Multiply_VectorByFloat(Vector=<70 1.1 1.1>, Float=$User.ProjectileScale)
    Mesh Scale Mode=::NewEnumerator4
    UsePositionOffset=true
    Write Material Random=true
   ) {module=InitializeParticle001}
   &DynamicMaterialParameters(
    Index 0 Param 4=&RandomRangeFloat(Minimum=0, Maximum=1, Evaluation Type=::NewEnumerator1)
    Module.Param0WriteEnabled=false
    Module.Param1WriteEnabled=false
    Module.Param2WriteEnabled=false
    Module.Param3WriteEnabled=true
   ) {module=DynamicMaterialParameters001}
   &InitialMeshOrientation(Orientation Axis=<0 0 1>, Orientation Vector=<0 0 1>, Use Orientation Vector=true, Use Rotation Vector=true)
  ]
  particleUpdateScript [
   &ParticleState()
   &DynamicMaterialParameters(
    Index 0 Param 1=&FloatFromCurve(FloatCurve=DataInterface() {propertyValues {Curve {keys [<at 0 0>,<at 1 0>]}}})
    Index 0 Param 2=&FloatFromCurve(FloatCurve=DataInterface() {propertyValues {Curve {keys [<at 0 1 RCIM_Cubic>, {arriveTangent=-4.4839, leaveTangent=-4.4839, time=1, value=1}]}}})
    Index 0 Param 3=1
    Module.Param0WriteEnabled=true
    Module.Param1WriteEnabled=true
    Module.Param2WriteEnabled=true
    Module.Param3WriteEnabled=false
   )
   &Color(Color=&LerpLinearColors(
     StartColor=$User.Color_01
     EndColor=$User.Color_01
     LerpFactor=&FloatFromCurve(FloatCurve=DataInterface() {propertyValues {Curve {keys [
          {interpMode=RCIM_Cubic, tangentMode=RCTM_User, time=0, value=0.5}
          {interpMode=RCIM_Cubic, tangentMode=RCTM_User, time=0.5, value=0}
          {tangentMode=RCTM_User, time=1, value=0.5}
         ]}}})
    ))
   &ScaleColor(
    Color Value To Scale=$Particles.Color
    Scale RGB=&VectorFromFloat(Value=1.2)
    ScaleA=true
    ScaleRGB=true
   ) {module=ScaleColor001}
   &ScaleColor(
    Color Value To Scale=$Particles.Color
    Scale Alpha=&FloatFromCurve(
     FloatCurve=DataInterface() {propertyValues {Curve {keys [
         {interpMode=RCIM_Cubic, tangentMode=RCTM_User, time=0, value=0}
         {interpMode=RCIM_Cubic, tangentMode=RCTM_User, time=0.5, value=1}
         {tangentMode=RCTM_User, time=1, value=0}
        ]}}}
     Scale Curve=0
    )
    Scale RGB=&VectorFromCurve(VectorCurve=DataInterface() {propertyValues {XCurve {keys [<at 0 1>,<at 1 1>]}, YCurve {keys [<at 0 1>,<at 1 1>]}, ZCurve {keys [<at 0 1>,<at 1 1>]}}})
    ScaleA=true
    ScaleRGB=false
   ) {enabled=false}
   &ScaleMeshSize(Scale Factor=&VectorFromFloat(Value=&FloatFromCurve(FloatCurve=DataInterface() {propertyValues {Curve {keys [
          {arriveTangent=4.1785, interpMode=RCIM_Cubic, leaveTangent=4.1785, tangentMode=RCTM_User, time=0, value=0.25}
          {arriveTangent=0.4581, interpMode=RCIM_Cubic, leaveTangent=0.4581, tangentMode=RCTM_User, time=0.2223, value=0.9}
          {arriveTangent=-0.0405, interpMode=RCIM_Cubic, leaveTangent=-0.0405, tangentMode=RCTM_User, time=1, value=1}
         ]}}}))) {enabled=false, module=ScaleMeshSize001}
   &ScaleMeshSize(Scale Factor=&VectorFromFloat(Value=&FloatFromCurve(FloatCurve=DataInterface() {propertyValues {Curve {keys [
          <at 0 1 RCIM_Cubic>
          {arriveTangent=0.25, interpMode=RCIM_Cubic, leaveTangent=0.25, time=0.1, value=0.95}
          <at 0.2 1.05 RCIM_Cubic>
          {arriveTangent=-0.25, interpMode=RCIM_Cubic, leaveTangent=-0.25, time=0.3, value=0.95}
          {arriveTangent=0.1167, interpMode=RCIM_Cubic, leaveTangent=0.1167, time=0.4, value=1}
          {arriveTangent=0.125, interpMode=RCIM_Cubic, leaveTangent=0.125, time=0.6, value=0.985}
          {arriveTangent=0.0375, interpMode=RCIM_Cubic, leaveTangent=0.0375, time=0.8, value=1.05}
          <at 1 1 RCIM_Cubic>
         ]}}}))) {module=ScaleMeshSize002}
   &ScaleMeshSize(Scale Factor=&VectorFromFloat(Value=&RandomRangeFloat(Minimum=0.99, Maximum=1.01, Evaluation Type=::NewEnumerator1))) {module=ScaleMeshSize003}
   &ScaleMeshSize(Scale Factor=&MakeVector(
     X=&FloatFromCurve(FloatCurve=DataInterface() {propertyValues {Curve {keys [<at 0 0>,<at 0.001 6>]}}}, CurveIndex=$Emitter.Age)
     Y=1
     Z=1
    )) {enabled=false}
   /SetVariables(
    Particles.Color_02:LinearColor=&MultiplyLinearColor(Linear Color=$Particles.Color, Scale Factor=<0.6 0.6 0.6 1>)
    Particles.MeshScale_02:Vector3f=&Multiply_Vector(A=$Particles.Scale, B=<1.25 0.95 0.95>)
   ) {isSetParameters=true}
  ]
  rendererClasses [&NiagaraMeshRendererProperties]
  renderers [
   {rendererClass=&NiagaraMeshRendererProperties, rendererIndex=0}
   {rendererClass=&NiagaraMeshRendererProperties, rendererIndex=1}
  ]
 }
 OuterCone {
  bEnabled=false
  emitterUpdateScript [
   &EmitterState()
   &SpawnRate(Use Spawn Probability=false)
  ]
  particleSpawnScript [
   &InitializeParticle(
    Color=<rgba 2 0.6 1.4 1>
    Mesh Scale=&Multiply_VectorByFloat(Vector=<1.8 1.2 1.2>, Float=0.7)
    Mesh Scale Mode=::NewEnumerator4
    Position Offset=<4 0 0>
    Sprite Size Mode=::NewEnumerator3
    UsePositionOffset=true
    Write Material Random=true
   )
   &InitialMeshOrientation(Orientation Vector=<-1 0 0>, Use Orientation Vector=true, Use Rotation Vector=true)
  ]
  particleUpdateScript [
   &ParticleState()
   &FN_Radial_Material(Base Sprite Tile V=2, Depth Fade=0, Radial Mask Power=3, Radial Mask Strength=10, Sprite Extrusion Speed=1.5)
   &ScaleColor(
    Scale Alpha=&FN_Select_Float_from_Time_Of_Day(Morning=1, Day=1, Evening=1, Night=0.5)
    Scale RGB=<4 2 2>
    ScaleA=true
    ScaleRGB=false
   )
  ]
  rendererClasses [&NiagaraMeshRendererProperties]
  renderers [{rendererClass=&NiagaraMeshRendererProperties, rendererIndex=0}]
 }
 OuterCone_Big {
  bEnabled=false
  emitterUpdateScript [
   &EmitterState()
   &SpawnRate(Use Spawn Probability=false)
  ]
  particleSpawnScript [
   &InitializeParticle(
    Color=<rgba 2 0.6 1.4 2>
    Mesh Scale=&Multiply_VectorByFloat(Vector=<2 1.1 1.1>, Float=0.7)
    Mesh Scale Mode=::NewEnumerator4
    Position Offset=<5 0 0>
    Sprite Size Mode=::NewEnumerator3
    UsePositionOffset=true
    Write Material Random=true
   )
   &InitialMeshOrientation(Orientation Vector=<-1 0 0>, Use Orientation Vector=true, Use Rotation Vector=true)
  ]
  particleUpdateScript [
   &ParticleState()
   &FN_Radial_Material(Base Sprite Tile V=2, Depth Fade=0, Radial Mask Power=3, Sprite Extrusion Speed=1.5)
  ]
  rendererClasses [&NiagaraMeshRendererProperties]
  renderers [{rendererClass=&NiagaraMeshRendererProperties, rendererIndex=0}]
 }
 OverAirglow {
  bEnabled=false
  emitterUpdateScript [
   &EmitterState(Life Cycle Mode=::NewEnumerator1, Loop Behavior=::NewEnumerator1, Loop Duration=2, UseLoopDelay=false)
   &SpawnBurst_Instantaneous(Spawn Count=1, Use Spawn Probability=false)
  ]
  particleSpawnScript [
   &InitializeParticle(
    Color=<rgba 0.4873 0.09 0.75 1>
    Mesh Scale Mode=::NewEnumerator3
    Mesh Uniform Scale=&Multiply_Float(A=$User.ProjectileScale, B=1)
    Sprite Size Mode=::NewEnumerator3
    Uniform Sprite Size=&Multiply_Float(A=$User.ProjectileScale, B=150)
    UsePositionOffset=false
    Write Material Random=true
   )
   &InitialMeshOrientation(Orientation Vector=<0 0 -1>, Use Orientation Vector=true, Use Rotation Vector=false)
  ]
  particleUpdateScript [
   &ParticleState(Kill Particles When Lifetime Has Elapsed=false)
   &DynamicMaterialParameters(
    Index 0 Param 1=1.5
    Index 0 Param 2=1
    Module.Param0WriteEnabled=true
    Module.Param1WriteEnabled=true
    Module.Param2WriteEnabled=true
    Module.Param3WriteEnabled=true
   )
   &ScaleColor(Scale Alpha=6, ScaleA=true, ScaleRGB=true)
   &CameraOffset(Camera Offset Amount=10)
   &ScaleSpriteSize(Scale Factor=&Vector2DFromFloat(Value=&FloatFromCurve(
      FloatCurve=DataInterface() {propertyValues {Curve {keys [<at 0 1>,<at 1 0>]}}}
      CurveIndex=&Lerp_Float(A=0, B=1, Alpha=0.5)
     )))
  ]
  rendererClasses [&NiagaraSpriteRendererProperties]
  renderers [{rendererClass=&NiagaraSpriteRendererProperties}]
 }
 Rings {
  emitterUpdateScript [
   &EmitterState()
   &SpawnRate(SpawnRate=64, Use Spawn Probability=false)
  ]
  particleSpawnScript [
   &InitializeParticle(
    Color=$User.Color_01
    Lifetime=0.65
    Mesh Scale Mode=::NewEnumerator3
    Mesh Uniform Scale=&Multiply_Float(A=$User.ProjectileScale, B=0.8)
    UsePositionOffset=true
    Write Material Random=true
   )
   &InitialMeshOrientation(
    Orientation Vector=<0 0 1>
    Rotation=&RandomRangeVector(Minimum=<0 0 0>, Maximum=<0 0 1>)
    Rotation Coordinate Space=::NewEnumerator0
    Use Orientation Vector=true
    Use Rotation Vector=false
   )
   &AddVelocity(
    Coordinate Space=::World
    Velocity=&Multiply_VectorByFloat(Vector=&RotateVectorByQuaternion(Vector To Rotate=<-2 0 0>, Quaternion=$Engine.Owner.Rotation), Float=1)
   )
  ]
  particleUpdateScript [
   &ParticleState()
   &SolveRotationalForcesAndVelocity() {enabled=false}
   /SetVariables(Particles.SizeWave:NiagaraFloat=&Waveform(
     Global Input Position=&Multiply_Float(A=$Engine.Time, B=1)
     Global Phase=$Particles.NormalizedAge
     "Waveform [1]"=::NewEnumerator0
     "[1] Phase Scale"=1.75
     "[1] Phase"=$Particles.NormalizedAge
     "[1] Frequency"=&Lerp_Float(A=0.4, B=1.1, Alpha=$User.ADS)
     "[1] Amplitude Min/Max"=<0 1>
     "[1] Scale"=0.5
     "[1] Bias"=0.5
     "[2] Phase"=$Emitter.Age
     "[3] Phase"=$Emitter.Age
    )) {enabled=false, isSetParameters=true}
   &SolveForcesAndVelocity(Clamp Velocity=false, Limit Acceleration=false)
   &ScaleColor(
    Scale Alpha=&FloatFromCurve(FloatCurve=DataInterface() {propertyValues {Curve {keys [<at 0 1>,<at 0.2 0>]}}}, Scale Curve=2)
    Scale RGB=&Multiply_VectorByFloat(Vector=<1 1 1>, Float=&Lerp_Float(A=2, B=3, Alpha=$User.ADS))
    ScaleA=true
    ScaleRGB=true
   )
   &DynamicMaterialParameters(
    Index 0 Param 1=&FloatFromCurve(
     FloatCurve=DataInterface() {propertyValues {Curve {keys [<at 0 0.7>,<at 1 0.2>]}}}
     CurveIndex=&FN_LOD_Distance_Remap(Start Distance=0, End Distance=100, Minimum Output=0, Maximum Output=1)
     Scale Curve=75
    )
    Module.Param0WriteEnabled=true
    Module.Param1WriteEnabled=true
    Module.Param2WriteEnabled=true
    Module.Param3WriteEnabled=false
   )
   &ScaleMeshSize(Scale Factor=&Multiply_VectorByFloat(
     Vector=&Multiply_VectorByFloat(
      Vector=&MakeVector(X=$Particles.SizeWave, Y=$Particles.SizeWave, Z=1)
      Float=&FloatFromCurve(FloatCurve=DataInterface() {propertyValues {Curve {keys [
           {interpMode=RCIM_Cubic, tangentMode=RCTM_User, time=0, value=1}
           {arriveTangent=-0.5603, interpMode=RCIM_Cubic, tangentMode=RCTM_Break, time=1, value=0.8}
          ]}}})
     )
     Float=$User.RingSize
    )) {enabled=false}
   &ScaleMeshSize(Scale Factor=&VectorFromFloat(Value=&FloatFromCurve(FloatCurve=DataInterface() {propertyValues {Curve {keys [
          {interpMode=RCIM_Cubic, tangentMode=RCTM_User, time=0, value=1}
          {arriveTangent=0.0182, interpMode=RCIM_Cubic, leaveTangent=0.0182, tangentMode=RCTM_User, time=0.2, value=1.35}
         ]}}}))) {module=ScaleMeshSize001}
   &ScaleMeshSize(Scale Factor=&VectorFromFloat(Value=&FloatFromCurve(FloatCurve=DataInterface() {propertyValues {Curve {keys [
          <at 0 1 RCIM_Cubic>
          {arriveTangent=0.25, interpMode=RCIM_Cubic, leaveTangent=0.25, time=0.1, value=0.95}
          <at 0.2 1.05 RCIM_Cubic>
          {arriveTangent=-0.25, interpMode=RCIM_Cubic, leaveTangent=-0.25, time=0.3, value=0.95}
          {arriveTangent=0.1167, interpMode=RCIM_Cubic, leaveTangent=0.1167, time=0.4, value=1}
          {arriveTangent=0.125, interpMode=RCIM_Cubic, leaveTangent=0.125, time=0.6, value=0.985}
          {arriveTangent=0.0375, interpMode=RCIM_Cubic, leaveTangent=0.0375, time=0.8, value=1.05}
          <at 1 1 RCIM_Cubic>
         ]}}}))) {enabled=false, module=ScaleMeshSize002}
   &ScaleMeshSize(Scale Factor=&VectorFromFloat(Value=&RandomRangeFloat(Minimum=0.99, Maximum=1.01, Evaluation Type=::NewEnumerator1))) {enabled=false, module=ScaleMeshSize003}
  ]
  rendererClasses [&NiagaraMeshRendererProperties]
  renderers [
   {rendererClass=&NiagaraMeshRendererProperties, rendererIndex=0}
   {rendererClass=&NiagaraMeshRendererProperties, rendererIndex=1}
  ]
 }
 Streaks {
  emitterUpdateScript [
   &EmitterState(Life Cycle Mode=::NewEnumerator1, Loop Behavior=::NewEnumerator1, Loop Duration=2, UseLoopDelay=false)
   &SpawnBurst_Instantaneous(Spawn Count=1, Use Spawn Probability=false)
  ]
  particleSpawnScript [&InitializeParticle(
    Color=$User.Color_01
    Sprite Rotation Mode=::NewEnumerator2
    Sprite Size Mode=::NewEnumerator3
    Uniform Sprite Size=&Multiply_Float(A=$User.ProjectileScale, B=375)
    UsePositionOffset=false
    Write Material Random=true
   )]
  particleUpdateScript [
   &ParticleState(Kill Particles When Lifetime Has Elapsed=false)
   &SpriteRotationRate(Rotation Rate=&RandomRangeFloat(Minimum=3.5, Maximum=8))
   &ScaleColor(
    Scale Alpha=&FN_Select_Float_from_Time_Of_Day(Morning=1, Day=1, Evening=1, Night=0.5)
    Scale RGB=<4 2 2>
    ScaleA=true
    ScaleRGB=true
   )
   &DynamicMaterialParameters(
    Index 0 Param 1=1
    Index 0 Param 4=1
    Module.Param0WriteEnabled=true
    Module.Param1WriteEnabled=true
    Module.Param2WriteEnabled=true
    Module.Param3WriteEnabled=true
   )
  ]
  rendererClasses [&NiagaraSpriteRendererProperties]
  renderers [{rendererClass=&NiagaraSpriteRendererProperties}]
 }
 Trail_Core {
  bEnabled=false
  emitterUpdateScript [
   &EmitterState()
   &SpawnRate(SpawnRate=40, Use Spawn Probability=false)
  ]
  particleSpawnScript [
   &InitializeRibbon(
    Color=$User.Color_03
    Lifetime=0.25
    Module.Write Color=true
    Module.Write Lifetime=true
    Module.Write Mass=true
    Module.Write Material Random=true
    Module.Write Position=false
    Module.Write RibbonTwist=false
    Module.Write RibbonWidth=true
    Ribbon Width=&Multiply_Float(A=$User.ProjectileScale, B=150)
   )
   &AddVelocity(Velocity=<-8000 0 0>)
   &SystemLocation(MaskSpawnGroup=false, Offset=<-10 0 0>)
  ]
  particleUpdateScript [
   &ParticleState()
   &CurlNoiseForce(
    Noise Frequency=10
    Noise Strength=&FloatFromCurve(
     FloatCurve=DataInterface() {propertyValues {Curve {keys [{tangentMode=RCTM_User, time=0, value=1}, {tangentMode=RCTM_User, time=1, value=0}]}}}
     Scale Curve=500
    )
    Pan Noise=true
    Pan Noise Field=&MakeVector(
     X=500
     Y=250
     Z=&FloatFromCurve(
      FloatCurve=DataInterface() {propertyValues {Curve {keys [{tangentMode=RCTM_User, time=0, value=0}, {tangentMode=RCTM_User, time=1, value=1}]}}}
      CurveIndex=$Emitter.NormalizedLoopAge
      Scale Curve=5
     )
    )
   ) {enabled=false, module=CurlNoiseForce001}
   &SolveForcesAndVelocity(Clamp Velocity=false, Limit Acceleration=false)
   &ScaleColor(
    Scale Alpha=&FloatFromCurve(FloatCurve=DataInterface() {propertyValues {Curve {keys [{interpMode=RCIM_Cubic, tangentMode=RCTM_User, time=0, value=1.5}, {interpMode=RCIM_Cubic, tangentMode=RCTM_User, time=1, value=0}]}}})
    Scale RGB=<1 0.6 0.6>
    ScaleA=true
    ScaleRGB=true
   )
   &ScaleRibbonWidth(Ribbon Width Scale=&FloatFromCurve(FloatCurve=DataInterface() {propertyValues {Curve {keys [{interpMode=RCIM_Cubic, tangentMode=RCTM_User, time=0, value=1}, {interpMode=RCIM_Cubic, tangentMode=RCTM_User, time=1, value=0.25}]}}}))
   &DynamicMaterialParameters(
    Index 0 Param 1=&RandomRangeFloat(Minimum=1, Maximum=1)
    Index 0 Param 2=1
    Index 0 Param 3=2
    Module.Param0WriteEnabled=true
    Module.Param1WriteEnabled=true
    Module.Param2WriteEnabled=true
    Module.Param3WriteEnabled=true
   )
  ]
  rendererClasses [&NiagaraRibbonRendererProperties]
  renderers [{rendererClass=&NiagaraRibbonRendererProperties}]
 }
 Trail_Dark {
  bEnabled=false
  emitterUpdateScript [
   &EmitterState()
   &SpawnRate(SpawnRate=40, Use Spawn Probability=false)
  ]
  particleSpawnScript [
   &InitializeRibbon(
    Color=$User.Color_03
    Lifetime=0.6
    Module.Write Color=true
    Module.Write Lifetime=true
    Module.Write Mass=true
    Module.Write Material Random=true
    Module.Write Position=false
    Module.Write RibbonTwist=false
    Module.Write RibbonWidth=true
    Ribbon Width=&Multiply_Float(A=$User.ProjectileScale, B=350)
   )
   &AddVelocity(Velocity=<-8000 0 0>)
   &SystemLocation(MaskSpawnGroup=false, Offset=<-10 0 0>)
  ]
  particleUpdateScript [
   &ParticleState()
   &CurlNoiseForce(
    Noise Frequency=10
    Noise Strength=&FloatFromCurve(
     FloatCurve=DataInterface() {propertyValues {Curve {keys [{tangentMode=RCTM_User, time=0, value=1}, {tangentMode=RCTM_User, time=1, value=0}]}}}
     Scale Curve=500
    )
    Pan Noise=true
    Pan Noise Field=&MakeVector(
     X=500
     Y=250
     Z=&FloatFromCurve(
      FloatCurve=DataInterface() {propertyValues {Curve {keys [{tangentMode=RCTM_User, time=0, value=0}, {tangentMode=RCTM_User, time=1, value=1}]}}}
      CurveIndex=$Emitter.NormalizedLoopAge
      Scale Curve=5
     )
    )
   ) {enabled=false}
   &SolveForcesAndVelocity(Clamp Velocity=false, Limit Acceleration=false)
   &ScaleColor(
    Scale Alpha=&FloatFromCurve(FloatCurve=DataInterface() {propertyValues {Curve {keys [{interpMode=RCIM_Cubic, tangentMode=RCTM_User, time=0, value=1}, {interpMode=RCIM_Cubic, tangentMode=RCTM_User, time=1, value=0}]}}})
    Scale RGB=<0.8 0.6 0.6>
    ScaleA=true
    ScaleRGB=true
   )
   &ScaleRibbonWidth(Ribbon Width Scale=&FloatFromCurve(FloatCurve=DataInterface() {propertyValues {Curve {keys [{interpMode=RCIM_Cubic, tangentMode=RCTM_User, time=0, value=1}, {interpMode=RCIM_Cubic, tangentMode=RCTM_User, time=1, value=0.25}]}}}))
   &DynamicMaterialParameters(
    Index 0 Param 1=&RandomRangeFloat(Minimum=0.8, Maximum=0.8)
    Index 0 Param 2=1
    Index 0 Param 3=2
    Module.Param0WriteEnabled=true
    Module.Param1WriteEnabled=true
    Module.Param2WriteEnabled=true
    Module.Param3WriteEnabled=true
   )
  ]
  rendererClasses [&NiagaraRibbonRendererProperties]
  renderers [{rendererClass=&NiagaraRibbonRendererProperties}]
 }
 Trail_Light {
  bEnabled=false
  emitterUpdateScript [
   &EmitterState()
   &SpawnRate(SpawnRate=40, Use Spawn Probability=false)
  ]
  particleSpawnScript [
   &InitializeRibbon(
    Color=$User.Color_01
    Lifetime=0.4
    Module.Write Color=true
    Module.Write Lifetime=true
    Module.Write Mass=true
    Module.Write Material Random=true
    Module.Write Position=false
    Module.Write RibbonTwist=false
    Module.Write RibbonWidth=true
    Ribbon Width=&Multiply_Float(A=$User.ProjectileScale, B=200)
   )
   &AddVelocity(Velocity=<-8000 0 0>)
   &SystemLocation(MaskSpawnGroup=false, Offset=<-10 0 0>)
  ]
  particleUpdateScript [
   &ParticleState()
   &CurlNoiseForce(
    Noise Frequency=10
    Noise Strength=&FloatFromCurve(
     FloatCurve=DataInterface() {propertyValues {Curve {keys [{tangentMode=RCTM_User, time=0, value=1}, {tangentMode=RCTM_User, time=1, value=0}]}}}
     Scale Curve=500
    )
    Pan Noise=true
    Pan Noise Field=&MakeVector(
     X=500
     Y=250
     Z=&FloatFromCurve(
      FloatCurve=DataInterface() {propertyValues {Curve {keys [{tangentMode=RCTM_User, time=0, value=0}, {tangentMode=RCTM_User, time=1, value=1}]}}}
      CurveIndex=$Emitter.NormalizedLoopAge
      Scale Curve=5
     )
    )
   ) {enabled=false}
   &SolveForcesAndVelocity(Clamp Velocity=false, Limit Acceleration=false)
   &ScaleColor(
    Scale Alpha=&FloatFromCurve(
     FloatCurve=DataInterface() {propertyValues {Curve {keys [{interpMode=RCIM_Cubic, tangentMode=RCTM_User, time=0, value=1}, {interpMode=RCIM_Cubic, tangentMode=RCTM_User, time=1, value=0}]}}}
     Scale Curve=2
    )
    Scale RGB=<0.8 0.6 0.6>
    ScaleA=true
    ScaleRGB=true
   )
   &ScaleRibbonWidth(Ribbon Width Scale=&FloatFromCurve(FloatCurve=DataInterface() {propertyValues {Curve {keys [{interpMode=RCIM_Cubic, tangentMode=RCTM_User, time=0, value=1}, {interpMode=RCIM_Cubic, tangentMode=RCTM_User, time=1, value=0.25}]}}}))
   &DynamicMaterialParameters(
    Index 0 Param 1=&RandomRangeFloat(Minimum=0.5, Maximum=0.5)
    Index 0 Param 2=1
    Index 0 Param 3=2
    Module.Param0WriteEnabled=true
    Module.Param1WriteEnabled=true
    Module.Param2WriteEnabled=true
    Module.Param3WriteEnabled=true
   )
  ]
  rendererClasses [&NiagaraRibbonRendererProperties]
  renderers [{rendererClass=&NiagaraRibbonRendererProperties}]
 }
}
summary {
 systemName=NS_Skyfire_KnockbackWeapon_Proj
 emitters [
  {
   bEnabled=false
   emitterName=Mesh
   rendererClasses [&NiagaraSpriteRendererProperties, &NiagaraMeshRendererProperties]
  }
  {emitterName=Halo, rendererClasses [&NiagaraSpriteRendererProperties]}
  {emitterName=Streaks, rendererClasses [&NiagaraSpriteRendererProperties]}
  {bEnabled=false, emitterName=Trail_Dark, rendererClasses [&NiagaraRibbonRendererProperties]}
  {bEnabled=false, emitterName=Trail_Core, rendererClasses [&NiagaraRibbonRendererProperties]}
  {bEnabled=false, emitterName=OuterCone, rendererClasses [&NiagaraMeshRendererProperties]}
  {bEnabled=false, emitterName=Trail_Light, rendererClasses [&NiagaraRibbonRendererProperties]}
  {bEnabled=false, emitterName=OverAirglow, rendererClasses [&NiagaraSpriteRendererProperties]}
  {bEnabled=false, emitterName=OuterCone_Big, rendererClasses [&NiagaraMeshRendererProperties]}
  {emitterName=Rings, rendererClasses [&NiagaraMeshRendererProperties]}
  {emitterName=MeshTrail, rendererClasses [&NiagaraMeshRendererProperties]}
 ]
 userVariables [
  User.ADS {type=NiagaraFloat, defaultValue=0, description=""}
  User.Color_01 {type=LinearColor, description="", defaultValue {struct=LinearColor, value=<rgba 0 1 0.8115 1>}}
  User.Color_02 {type=LinearColor, description="", defaultValue {struct=LinearColor, value=<rgba 1 0.0396 0.894 1>}}
  User.Color_03 {type=LinearColor, description="", defaultValue {struct=LinearColor, value=<rgba 0.0664 0.2855 0.625 1>}}
  User.ProjectileScale {type=NiagaraFloat, defaultValue=0.1, description=""}
  User.RingSize {type=NiagaraFloat, defaultValue=0, description=""}
 ]
}
