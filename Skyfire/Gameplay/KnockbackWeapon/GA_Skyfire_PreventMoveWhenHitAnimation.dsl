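; Gameplay ability for the Skyfire knockback weapon that plays a stunned animation, named for preventing movement when hit.
; On activation it runs the parent logic, then plays the montage from GetStunnedMontage through a montage-and-wait task with no trigger name.
; The ability ends when the montage completes or is cancelled.
; No movement blocking appears in this graph, and the end event only forwards to the parent.
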(event Ability|EventActivateAbility
  ;[1 0.002934 0.9635] Play Stunned Animation
  (|CallParentFunction)
  (Ability|Tasks|PlayMontageandWaitforTrigger "None" (Variables|Default|GetStunnedMontage)
    (:Cancelled
      _)
    (:Completed
      _))
  (Ability|EndAbility))

(event Ability|EventOnEndAbility
  (params (bWasCancelled :type Boolean))
  ;[1 0.002934 0.9635] Play Stunned Animation
  (|CallParentFunction bWasCancelled))
