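; Gameplay ability that clears a player's stacking knockback count once they finish falling, for the Skyfire knockback weapon.
; On activation it stores the avatar pawn and binds to its landed and enter-water events.
; Either event removes all stacks of the knockback count effect, then removes the effect that granted this ability so it can be reapplied later, and ends the ability.
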(event Ability|EventActivateAbility
  ;[0.06801 1 0.5651] Bind to landed and enter water so we know when the player is done falling.
  (|CallParentFunction)
  (bind _outfortplayerpawn (CallFunction|GetAvatarasValidFortPlayerPawn)
    (:ValidFortPlayerPawn
      (bind _output_get (Variables|Default|SetPlayerPawn _outfortplayerpawn))
      (EventDispatchers|BindEventtoOnPawnLanded _output_get (EventDispatchers|CreateEvent :!event Custom|PawnLanded_Ground))
      (EventDispatchers|BindEventtoOnPawnEnterWater _output_get (EventDispatchers|CreateEvent :!event Custom|PawnLanded_Water)))))

(event Custom|PawnLanded_Water
  (params (FortPlayerPawn :type Fort%20Player%20Pawn%20Object%20Reference))
  (bind _playerpawn (Variables|Default|GetPlayerPawn))
  (bind _returnvalue (Ability|GetAbilitySystemComponent _playerpawn))
  ;[0 0.6896 1] Remove all knockback stacks and remove the GE that applied this GA so it can be reapplied properly later.
  (Utilities|IsValid _playerpawn
    (:Is%20Valid
      (Utilities|IsValid _returnvalue
        (:Is%20Valid
          (GameplayEffects|RemoveActiveGameplayEffectbySourceEffect _returnvalue "/Skyfire/Gameplay/KnockbackWeapon/GE_Skyfire_StackingKnockbackCount.GE_Skyfire_StackingKnockbackCount_C" 0 -1)
          (Ability|RemoveGrantedbyEffect)
          (Ability|EndAbility))))))

(event Custom|PawnLanded_Ground
  (params (Hit :type Hit%20Result%20Structure%20%28by%20ref%29))
  (bind _playerpawn (Variables|Default|GetPlayerPawn))
  (bind _returnvalue (Ability|GetAbilitySystemComponent _playerpawn))
  ;[0 0.6896 1] Remove all knockback stacks and remove the GE that applied this GA so it can be reapplied properly later.
  (Utilities|IsValid _playerpawn
    (:Is%20Valid
      (Utilities|IsValid _returnvalue
        (:Is%20Valid
          (GameplayEffects|RemoveActiveGameplayEffectbySourceEffect _returnvalue "/Skyfire/Gameplay/KnockbackWeapon/GE_Skyfire_StackingKnockbackCount.GE_Skyfire_StackingKnockbackCount_C" 0 -1)
          (Ability|RemoveGrantedbyEffect)
          (Ability|EndAbility))))))
