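; Gameplay ability that removes the knockback weapon from the player who owns it.
; On activation it calls the parent, then, only with authority, gets the avatar actor's Fort player controller.
; If that controller is valid, it removes the knockback weapon item from that player's inventory and ends the ability.
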(event Ability|EventActivateAbility
  (bind _returnvalue (Ability|GetAvatarActorfromActorInfo))
  (bind _returnvalue_2 (Fortnite|GetFortPlayerControllerfromActor _returnvalue))
  ;[0.03212 1 0] Remove Weapon from Player
  (|CallParentFunction)
  (Ability|HasAuthority
    (:True
      (Utilities|IsValid _returnvalue_2
        (:Is%20Valid
          (Inventory|RemoveItemFromPlayer _returnvalue_2 (Variables|Default|GetKnockbackWeapon))
          (Ability|EndAbility))))))
