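# Lit alien-metal surface material for the Skyfire knockback weapon, usable on skeletal meshes.
# Packed masks, blended base and detail normals, and tiled tech, dirt and metal patterns drive the PBR channels, with cheap world-aligned UVs selected by rounded vertex normal.
# Fresnel and sine banding shift the curve-atlas base color with view angle, while reflection-vector curve lookups build an anodized chroma effect.
# A static switch selects time-animated roughness blocks.
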Material /Skyfire/Gameplay/KnockbackWeapon/Materials/M_Skyfire_Alien_Metal_Weapon
graph {
 properties {bUsedWithMeshDeformer=true, bUsedWithSkeletalMesh=true, pixelDepthOffsetMode=PDOM_Legacy, usedShadingModels=Default Lit}
 nodes [
  LinearInterpolate(
   A=ScalarParameter() {defaultValue=0.2, parameterName=FakeLightLow}
   B=ScalarParameter() {defaultValue=1, parameterName=FakeLightHigh}
   Alpha=SphereMask(
    A=VectorParameter() {parameterName=Center, defaultValue=<rgba -12063 -7732 160 0>}.RGB
    B=WorldPosition().XYZ
    Radius=ScalarParameter() {defaultValue=350, parameterName=FakeLightRadius}
    Hardness=ScalarParameter() {defaultValue=0.5, parameterName=FakeLightHardness}
   )
  )
  n1 = TextureSampleParameter2D() {parameterName=Masks, samplerType=SAMPLERTYPE_LinearColor, texture=/Game/Environments/Apollo/Sets/Skyfire/Textures/T_rearwall_MSK}
  NamedRerouteDeclaration(None=@n1.RGB) {name=MasksRGB, nodeColor=<rgba 0 0.0391 1 1>}
  NamedRerouteDeclaration(None=@n1.R) {name=MasksR_Glow, nodeColor=<rgba 0.2656 0 1 1>}
  n2 = WorldPosition()
  n3 = ComponentMask(None=@n2.XYZ) {b=true, r=true}
  n4 = ComponentMask(None=@n2.XYZ) {b=true, g=true}
  n5 = VertexNormalWS()
  n6 = MaterialFunctionCall(Float3=Round(None=Abs(None=@n5))) {materialFunction=&BreakOutFloat3Components}
  n7 = NamedRerouteDeclaration(None=LinearInterpolate(A=LinearInterpolate(A=@n3, B=@n4, Alpha=@n6.R), B=ComponentMask(None=@n2.XYZ) {g=true, r=true}, Alpha=@n6.B)) {name=CheapWorldAlignedUVs, nodeColor=<rgba 0 0.7656 1 1>}
  NamedRerouteDeclaration(None=Divide(A=@n7, B=ScalarParameter() {defaultValue=220, group=Tile, parameterName=TechTileDivisor})) {name=TechPatternUVs, nodeColor=<rgba 0.0781 1 0 1>}
  NamedRerouteDeclaration(None=@n1.RGBA) {name=MasksRGBA, nodeColor=<rgba 1 0.6094 0 1>}
  n8 = TextureSampleParameter2D() {group=Normal, parameterName=Normal, samplerType=SAMPLERTYPE_Normal, texture=/Game/Environments/Apollo/Sets/Skyfire/Textures/T_rear_wall_N}
  n9 = TextureSampleParameter2D(UVs=NamedRerouteUsage()) {
   group=Normal
   parameterName=DetailNormal
   samplerType=SAMPLERTYPE_Normal
   texture=/Game/Global/Textures/Default/Tileable_Patterns/T_BrushedMetal_Detail_N
  }
  n10 = NamedRerouteDeclaration(None=@n9.RGB) {name=DetailNormal, nodeColor=<rgba 0 1 0.9531 1>}
  n11 = VectorParameter() {group=Normal, parameterName=DetailNormalStrength, defaultValue=<rgba 0.3 0.3 1 0>}
  NamedRerouteDeclaration(None=MaterialFunctionCall(BaseNormal=@n8.RGB, AdditionalNormal=Multiply(A=@n10, B=@n11.RGB)) {materialFunction=&BlendAngleCorrectedNormals}.Result) {name=FinalNormalMap, nodeColor=<rgba 0.7109 1 0 1>}
  n12 = Multiply(
   A=NamedRerouteUsage()
   B=VectorParameter() {group=Emissive, parameterName=CurvatureColorEmissive, defaultValue=<rgba 0.3125 0.0374 0.1936 0>}.RGB
  )
  n13 = NamedRerouteUsage()
  n14 = NamedRerouteUsage()
  n15 = Multiply(A=NamedRerouteUsage(), B=VectorParameter() {group=Emissive, parameterName=GlowColor}.RGB)
  NamedRerouteDeclaration(None=Add(A=Add(A=@n12, B=@n13), B=Multiply(A=@n14, B=@n15))) {name=Emissive, nodeColor=<rgba 0 0.25 1 1>}
  n16 = ScalarParameter() {defaultValue=0.75, group=Base Color, parameterName=BaseColorDirtLow}
  n17 = ScalarParameter() {defaultValue=1.4, group=Base Color, parameterName=BaseColorDirtHigh}
  n18 = TextureSampleParameter2D(UVs=NamedRerouteUsage()) {parameterName=PackB, samplerType=SAMPLERTYPE_LinearColor, texture=/Game/Environments/Apollo/Sets/Skyfire/Textures/T_Skyfire_PackB}
  n19 = MaterialFunctionCall() {materialFunction=&MF_Fresnel_FX}
  n20 = TextureSampleParameter2D(UVs=NamedRerouteUsage()) {parameterName=PackA, samplerType=SAMPLERTYPE_LinearColor, texture=/Game/Environments/Apollo/Sets/Skyfire/Textures/T_Skyfire_PackA}
  n21 = NamedRerouteDeclaration(None=@n20.R) {name=TechPatternBlocks, nodeColor=<rgba 0 1 0.8359 1>}
  n22 = ConstantBiasScale(None=Sine(None=Add(A=@n19.Fresnel, B=@n21)) {period=0.5})
  n23 = NamedRerouteDeclaration(None=LinearInterpolate(
    A=ScalarParameter() {defaultValue=0.85, group=Base Color, parameterName=ViewAngleDependantTechPattern_Low}
    B=ScalarParameter() {defaultValue=1, group=Base Color, parameterName=ViewAngleDependantTechPattern_High}
    Alpha=Multiply(A=@n19.Fresnel Inverted, B=@n22)
   )) {name=ViewAngleDependantTechPattern, nodeColor=<rgba 1 0 0.6797 1>}
  n24 = CurveAtlasRowParameter(CurveTime=NamedRerouteUsage()) {
   atlas=/Game/Environments/Apollo/Sets/Skyfire/Curves/SkyfireCurveAtlas
   curve=/Game/Environments/Apollo/Sets/Skyfire/Curves/C_Skyfire_Diffuse_01
   group=Base Color
   parameterName=DiffuseColorCurve
  }
  NamedRerouteDeclaration(None=Multiply(A=LinearInterpolate(A=@n16, B=@n17, Alpha=@n18.G), B=Multiply(A=@n23, B=@n24))) {name=BaseColor, nodeColor=<rgba 0 0.8828 1 1>}
  NamedRerouteDeclaration(None=Divide(A=NamedRerouteUsage(), B=ScalarParameter() {defaultValue=400, group=Tile, parameterName=DirtTileDivisor})) {name=DirtUVs, nodeColor=<rgba 1 0.6797 0 1>}
  n25 = NamedRerouteUsage()
  n26 = ScalarParameter() {defaultValue=1.5, group=Ambient Occlusion, parameterName=AO_Power}
  n27 = ScalarParameter() {defaultValue=1.5, group=Ambient Occlusion, parameterName=AO_Strength}
  n28 = NamedRerouteUsage()
  NamedRerouteDeclaration(None=LinearInterpolate(B=Power(Base=@n25, Exp=@n26), Alpha=Max(A=@n27, B=@n28)) {constA=1}) {name=AmbientOcclusion, nodeColor=<rgba 1 0.7969 0 1>}
  NamedRerouteDeclaration(None=@n1.G) {name=MasksG_Height, nodeColor=<rgba 0 1 0 1>}
  NamedRerouteDeclaration(None=@n1.B) {name=MasksB_AO, nodeColor=<rgba 0 0 1 1>}
  NamedRerouteDeclaration(None=@n1.A) {name=MasksA_CurvatureFine, nodeColor=<rgba 1 1 1 1>}
  NamedRerouteDeclaration(None=LinearInterpolate(
    A=ScalarParameter() {defaultValue=0.65, group=Metallic, parameterName=Metallic_Low}
    B=ScalarParameter() {defaultValue=0.4, group=Metallic, parameterName=Metallic_High}
    Alpha=NamedRerouteUsage()
   )) {name=Metallic, nodeColor=<rgba 1 0 0.4453 1>}
  n29 = Add(A=Divide(A=Time(), B=ScalarParameter() {defaultValue=18, group=Roughness, parameterName=BlocksRoughnessTimeScale}), B=NamedRerouteUsage())
  n30 = TextureSampleParameter2D(UVs=NamedRerouteUsage()) {parameterName=PackA, samplerType=SAMPLERTYPE_LinearColor, texture=&T_Skyfire_PackA}
  n31 = StaticSwitchParameter(True=ConstantBiasScale(None=Sine(None=@n29) {period=0.4}), False=@n30.B) {group=Roughness, parameterName=UseAnimatedRoughnesMask}
  n32 = TextureSampleParameter2D(UVs=NamedRerouteUsage()) {parameterName=PackA, samplerType=SAMPLERTYPE_LinearColor, texture=&T_Skyfire_PackA}
  NamedRerouteDeclaration(None=LinearInterpolate(
    A=ScalarParameter() {defaultValue=0.28, group=Roughness, parameterName=Roughness_Low}
    B=ScalarParameter() {defaultValue=0.45, group=Roughness, parameterName=Roughness_High}
    Alpha=Multiply(A=@n31, B=@n32.A)
   )) {name=Roughness, nodeColor=<rgba 0.875 1 0 1>}
  n33 = Power(Base=NamedRerouteUsage()) {constExponent=4}
  n34 = Subtract(A=ReflectionVectorWS()) {constB=-0.5}
  n35 = CurveAtlasRowParameter(CurveTime=DotProduct(A=ComponentMask(None=@n34) {g=true, r=true}, B=Constant() {r=0.5})) {
   atlas=&SkyfireCurveAtlas
   curve=/Game/Environments/Apollo/Sets/Skyfire/Curves/C_Skyfire_AnodizedReflectionColor
   defaultValue=1
   group=Emissive
   parameterName=AnodizedColorCurve
  }
  n36 = Multiply(A=NamedRerouteUsage(), B=@n35)
  n37 = MaterialFunctionCall(Exponent=ScalarParameter() {defaultValue=2, group=Emissive, parameterName=ChromaFresnelExponent}) {materialFunction=&MF_Fresnel_FX}
  n38 = Multiply(A=NamedRerouteUsage(), B=ScalarParameter() {defaultValue=10, group=Emissive, parameterName=ChromaEdgeBoost})
  NamedRerouteDeclaration(None=Multiply(A=Multiply(A=@n33, B=@n36), B=Add(A=@n37.Fresnel Inverted, B=@n38))) {name=Chroma, nodeColor=<rgba 0.5 0 1 1>}
  n39 = ScalarParameter() {defaultValue=400, group=Tile, parameterName=MetalTileDivisorU}
  n40 = ScalarParameter() {defaultValue=50, group=Tile, parameterName=MetalTileDivisorV}
  NamedRerouteDeclaration(None=Divide(A=@n7, B=AppendVector(A=@n39, B=@n40))) {name=MetalUVs, nodeColor=<rgba 0 0.0156 1 1>}
  NamedRerouteUsage()
  NamedRerouteUsage()
  NamedRerouteUsage()
  NamedRerouteUsage()
  NamedRerouteUsage()
  NamedRerouteUsage()
  NamedRerouteUsage()
  ScalarParameter() {defaultValue=1, group=Specular, parameterName=Specular}
  ScalarParameter() {defaultValue=0.15, group=Anisotropy, parameterName=Anisotropy}
 ]
}
